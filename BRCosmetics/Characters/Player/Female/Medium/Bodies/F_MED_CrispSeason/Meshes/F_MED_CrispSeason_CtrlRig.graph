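# Control Rig for the F_MED_CrispSeason female body cosmetic, procedurally correcting body and accessory bones as the character moves.
# A shared DeformRig function runs first, then arm twist fixes and thigh spherical pose readers adjust belt and master_Art bones.
# A sequence offsets ProcMove photo controls from spine, clavicle, thigh and distance drivers.
# It then fits front and back dyn_photo chains with Spline IK, refined by weighted parent constraints and offsets.
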ControlRigBlueprint /BRCosmetics/Characters/Player/Female/Medium/Bodies/F_MED_CrispSeason/Meshes/F_MED_CrispSeason_CtrlRig
graphs [
 RigVMFunctionLibrary {links [], nodes []}
 RigVMModel {nodes [
   RigUnit_BeginExecution = &RigUnit_BeginExecution()
   FN_DeformRig_FNC_1 = RigVMFunctionReferenceNode(
    DeformRig_ Curve="(deformRig_l_upperarm_up_alpha_72_#1=False,deformRig_r_upperarm_up_alpha_73_#2=False,deformRig_l_upperarm_fr_alpha_74_#3=False,deformRig_r_upperarm_fr_alpha_75_CB51B8 ...(truncated)"
    DeformRig_Bool="(upperarm_1_#4=True,elbow_3_#5=True,wrist_7_#6=True,groin_8_#7=true,glute_17_#8=true,knee_9_40C408324CA261 ...(truncated)"
    DeformRig_Debug=(upperarm_debug_18_#4=False,wrist_debug_20_#6=False,lat_debug_26_#9=False,groinGlute_debug_27_#7=False)
    DeformRig_Value="(upperarm_up_18_#10=(Rotation=(X=0,Y=0,Z=0,W=1),Translation=(X=0,Y=0,Z=8),Scale3D=(X=0.7,Y=1,Z=1)),upperarm_fwd_fr_19_#11=(Rotation=(X=0,Y=0,Z=0,W=1),Translation=(X=0,Y=10,Z=0),Scale3D=(X=0.5,Y=1,Z= ...(truncated)"
    ExecuteContext=@RigUnit_BeginExecution.ExecutePin
   )
   n1 = &RigUnit_GetRelativeTransformForItem(Child=(Type=Bone,Name="upperarm_twist_02_l"), Parent=(Type=Bone,Name="upperarm_l"), bChildInitial=False, bParentInitial=False)
   Absolute = &RigVMFunction_MathDoubleAbs(Value=&RigVMFunction_MathQuaternionToEuler(RotationOrder=ZYX, Value=@n1.RelativeTransform.Rotation).Result.X)
   > &RigUnit_ModifyTransforms(
    ItemToModify=((Item=(Type=Bone,Name="upperarm_twist_02_l"),Transform=(Rotation=(X=0,Y=0,Z=0,W=1),Translation=(X=0,Y=0,Z=0),Scale3D=(X=1,Y=1.3,Z=1.3))))
    WeightMinimum=0
    Weight=&RigVMFunction_MathDoubleRemap(SourceMaximum=50, SourceMinimum=10, TargetMaximum=1, TargetMinimum=0, Value=@Absolute.Result).Result
   )
   n2 = &RigUnit_GetRelativeTransformForItem(Child=(Type=Bone,Name="upperarm_twist_02_r"), Parent=(Type=Bone,Name="upperarm_r"), bChildInitial=False, bParentInitial=False)
   n3 = &RigVMFunction_MathQuaternionToEuler(RotationOrder=ZYX, Value=@n2.RelativeTransform.Rotation)
   > &RigUnit_ModifyTransforms(
    ItemToModify=((Item=(Type=Bone,Name="upperarm_twist_02_r"),Transform=(Rotation=(X=0,Y=0,Z=0,W=1),Translation=(X=0,Y=0,Z=0),Scale3D=(X=1,Y=1.3,Z=1.3))))
    WeightMinimum=0
    Weight=&RigVMFunction_MathDoubleRemap(
     SourceMaximum=50
     SourceMinimum=10
     TargetMaximum=1
     TargetMinimum=0
     Value=&RigVMFunction_MathDoubleAbs(Value=@n3.Result.X).Result
    ).Result
   )
   RigUnit_GetRelativeTransformForItem = &RigUnit_GetRelativeTransformForItem(Child=(Type=Bone,Name="hand_l"), Parent=(Type=Null,Name="hand_l_InitPose"), bChildInitial=False, bParentInitial=False)
   > &RigUnit_ModifyTransforms(
    ItemToModify="((Item=(Type=Bone,Name=\"lowerarm_twist_02_l\"),Transform=(Rotation=(X=-0.1305,Y=0,Z=0,W=0.9914),Translation=(X=0,Y=0,Z=0),Scale3D=(X=1,Y=1.2,Z=1.2))),(Item=(Type=Bone,Name=\"lowerarm_twist_01_l\"),Transform=(Rotation=(X=-0.2588,Y=0,Z=0,W=0.9659),Translation=( ...(truncated)"
    WeightMinimum=0
    Weight=&RigVMFunction_MathDoubleRemap(
     SourceMaximum=130
     SourceMinimum=90
     TargetMaximum=1
     TargetMinimum=0
     bClamp=False
     Value=&RigVMFunction_MathQuaternionToEuler(RotationOrder=XYZ, Value=@RigUnit_GetRelativeTransformForItem.RelativeTransform.Rotation).Result.X
    ).Result
   )
   > n4 = &RigUnit_SphericalPoseReader(
    ActiveRegionScaleFactors=(PositiveWidth=5,NegativeWidth=0.5,PositiveHeight=5,NegativeHeight=5)
    ActiveRegionSize=0.07
    Debug=(bDrawDebug=false,bDraw2D=False,bDrawLocalAxes=False,DebugScale=25,DebugSegments=20,DebugThickness=0.25)
    DriverAxis=(X=-1,Y=0,Z=0)
    DriverItem=(Type=Bone,Name="thigh_l")
    FalloffRegionScaleFactors=(PositiveWidth=1,NegativeWidth=1.2,PositiveHeight=1,NegativeHeight=1)
    FalloffSize=0.4
    FlipHeightScaling=False
    FlipWidthScaling=False
    OptionalParentItem=(Type=Bone,Name="None")
    RotationOffset=(X=130,Y=0,Z=2.1195)
   )
   > &RigUnit_ModifyTransforms(
    ItemToModify="((Item=(Type=Bone,Name=\"bn_belt06\"),Transform=(Rotation=(X=-0.2548,Y=-0.0544,Z=-0.1449,W=-0.9545),Translation=(X=0.8471,Y=-7.0761,Z=-0.4605),Scale3D=(X=1,Y=1,Z=1))),(Item=(Type=Bone,Name=\"bn_belt05\"),Transform=(Rotation=(X=-0.3737,Y=-0.0404,Z=0.2129,W=-0.9 ...(truncated)"
    WeightMinimum=0
    Weight=@n4.OutputParam
   )
   > n5 = &RigUnit_SphericalPoseReader(
    ActiveRegionScaleFactors=(PositiveWidth=3,NegativeWidth=0.5,PositiveHeight=3,NegativeHeight=3)
    ActiveRegionSize=0.07
    Debug=(bDrawDebug=false,bDraw2D=False,bDrawLocalAxes=False,DebugScale=25,DebugSegments=20,DebugThickness=0.25)
    DriverAxis=(X=1,Y=0,Z=0)
    DriverItem=(Type=Bone,Name="thigh_r")
    FalloffRegionScaleFactors=(PositiveWidth=1,NegativeWidth=1.2,PositiveHeight=1,NegativeHeight=1)
    FalloffSize=0.3
    FlipHeightScaling=False
    FlipWidthScaling=False
    OptionalParentItem=(Type=Bone,Name="None")
    RotationOffset=(X=-106.4097,Y=-168.9433,Z=-10.3433)
   )
   > &RigUnit_ModifyTransforms(
    ItemToModify="((Item=(Type=Bone,Name=\"bn_belt04\"),Transform=(Rotation=(X=0,Y=0,Z=-0.1665,W=-0.986),Translation=(X=0,Y=0,Z=0),Scale3D=(X=1,Y=1,Z=1))),(Item=(Type=Bone,Name=\"bn_belt03\"),Transform=(Rotation=(X=-0.3999,Y=-0.1754,Z=-0.1184,W=-0.8918),Translation=(X=-0.1474,Y ...(truncated)"
    WeightMinimum=0
    Weight=@n5.OutputParam
   )
   > n6 = &RigUnit_SphericalPoseReader(
    ActiveRegionScaleFactors=(PositiveWidth=3,NegativeWidth=0.5,PositiveHeight=2,NegativeHeight=2)
    ActiveRegionSize=0.07
    Debug=(bDrawDebug=false,bDraw2D=False,bDrawLocalAxes=False,DebugScale=25,DebugSegments=20,DebugThickness=0.25)
    DriverAxis=(X=-1,Y=0,Z=0)
    DriverItem=(Type=Bone,Name="thigh_l")
    FalloffRegionScaleFactors=(PositiveWidth=1,NegativeWidth=1,PositiveHeight=1,NegativeHeight=1)
    FalloffSize=0.2
    FlipHeightScaling=False
    FlipWidthScaling=False
    OptionalParentItem=(Type=Bone,Name="None")
    RotationOffset=(X=184.2986,Y=13.6736,Z=5.1037)
   )
   > &RigUnit_ModifyTransforms(
    ItemToModify="((Item=(Type=Bone,Name=\"bn_belt05\"),Transform=(Rotation=(X=0,Y=0,Z=0.0506,W=0.9987),Translation=(X=0,Y=0.6993,Z=0),Scale3D=(X=1,Y=1,Z=1))),(Item=(Type=Bone,Name=\"bn_belt06\"),Transform=(Rotation=(X=0.0374,Y=0.0859,Z=-0.1839,W=0.9785),Translation=(X=-1.4727, ...(truncated)"
    WeightMinimum=0
    Weight=@n6.OutputParam
   )
   > n7 = &RigUnit_SphericalPoseReader(
    ActiveRegionScaleFactors=(PositiveWidth=3,NegativeWidth=0.5,PositiveHeight=3,NegativeHeight=2)
    ActiveRegionSize=0.07
    Debug=(bDrawDebug=false,bDraw2D=False,bDrawLocalAxes=false,DebugScale=25,DebugSegments=20,DebugThickness=0.25)
    DriverAxis=(X=-1,Y=0,Z=0)
    DriverItem=(Type=Bone,Name="thigh_r")
    FalloffRegionScaleFactors=(PositiveWidth=1,NegativeWidth=1,PositiveHeight=1,NegativeHeight=3)
    FalloffSize=0.2
    FlipHeightScaling=False
    FlipWidthScaling=False
    OptionalParentItem=(Type=Bone,Name="None")
    RotationOffset=(X=184.2986,Y=12.3155,Z=1.0785)
   )
   > &RigUnit_ModifyTransforms(
    ItemToModify="((Item=(Type=Bone,Name=\"bn_belt04\"),Transform=(Rotation=(X=0,Y=0,Z=0.083,W=0.9965),Translation=(X=-0.0037,Y=0.0223,Z=0),Scale3D=(X=1,Y=1,Z=1))),(Item=(Type=Bone,Name=\"bn_belt03\"),Transform=(Rotation=(X=0.0971,Y=-0.0143,Z=0.2754,W=0.9563),Translation=(X=0.0 ...(truncated)"
    WeightMinimum=0
    Weight=@n7.OutputParam
   )
   > n8 = &RigUnit_SphericalPoseReader(
    ActiveRegionScaleFactors=(PositiveWidth=3,NegativeWidth=1,PositiveHeight=3,NegativeHeight=1)
    ActiveRegionSize=0.07
    Debug=(bDrawDebug=false,bDraw2D=false,bDrawLocalAxes=False,DebugScale=25,DebugSegments=20,DebugThickness=0.25)
    DriverAxis=(X=-1,Y=0,Z=0)
    DriverItem=(Type=Bone,Name="thigh_l")
    FalloffRegionScaleFactors=(PositiveWidth=3,NegativeWidth=1.5,PositiveHeight=3,NegativeHeight=1.3)
    FalloffSize=0.3
    FlipHeightScaling=False
    FlipWidthScaling=False
    OptionalParentItem=(Type=Bone,Name="pelvis")
    RotationOffset=(X=145,Y=0,Z=30)
   )
   Power = /Script/RigVM.RigVMFunction_MathDoublePow(B=1.3, A=@n8.OutputParam) {is=struct RigVMUnitNode, template="Power::Execute(in A,in B,out Result)"}
   > &RigUnit_ModifyTransforms(
    ItemToModify=((Item=(Type=Bone,Name="master_Art"),Transform=(Rotation=(X=-0.4061,Y=-0.1036,Z=-0.0388,W=-0.9071),Translation=(X=-0.6311,Y=-3.1453,Z=0.1693),Scale3D=(X=1,Y=1,Z=1))))
    WeightMinimum=0
    Weight=@Power.Result
   )
   > n9 = &RigUnit_SphericalPoseReader(
    ActiveRegionScaleFactors=(PositiveWidth=2,NegativeWidth=1.8,PositiveHeight=2,NegativeHeight=3)
    ActiveRegionSize=0.07
    Debug=(bDrawDebug=true,bDraw2D=false,bDrawLocalAxes=False,DebugScale=25,DebugSegments=20,DebugThickness=0.25)
    DriverAxis=(X=-1,Y=0,Z=0)
    DriverItem=(Type=Bone,Name="thigh_l")
    FalloffRegionScaleFactors=(PositiveWidth=2,NegativeWidth=1,PositiveHeight=1,NegativeHeight=2)
    FalloffSize=0.35
    FlipHeightScaling=False
    FlipWidthScaling=False
    OptionalParentItem=(Type=Bone,Name="None")
    RotationOffset=(X=210,Y=-20,Z=-2.2821)
   )
   > n10 = &RigUnit_ModifyTransforms(
    ItemToModify=((Item=(Type=Bone,Name="master_Art"),Transform=(Rotation=(X=-0.482,Y=0.0928,Z=0.0514,W=-0.8697),Translation=(X=0.0001,Y=-6.0057,Z=-2.5491),Scale3D=(X=1,Y=1,Z=1))))
    WeightMinimum=0
    Weight=@n9.OutputParam
   )
   n11 = &RigVMFunction_Sequence(ExecuteContext=@n10.ExecutePin)
   ItemArray_2 = RigVMDispatchNode(Value="((Type=Bone,Name=\"dyn_photo_clavicle_front\"),(Type=Bone,Name=\"dyn_photo_front_01\"),(Type=Bone,Name=\"dyn_photo_front_02\"),(Type=Bone,Name=\"dyn_photo_front_03\"),(Type=Bone,Name=\"dyn_photo_front_04\"),(Type=Bone,Name=\"dyn_photo_front_05\"),(Type=Bone,Name=\"dyn_ ...(truncated)") {template="DISPATCH_RigVMDispatch_Constant(io Value)"}
   ItemArray_3 = RigVMDispatchNode(Value=((Type=Bone,Name="ProcMove_photo_clavicle_front"),(Type=Bone,Name="ProcMove_photo_front_02"),(Type=Bone,Name="ProceMove_photo_front_05"),(Type=Bone,Name="ProcMove_photo_front_09"))) {template="DISPATCH_RigVMDispatch_Constant(io Value)"}
   SplineIK = RigVMFunctionReferenceNode(
    Primary Axis=(X=1,Y=0,Z=0)
    Samples Per Segment=16
    Secondary Spline Direction=(X=0,Y=0,Z=6)
    Spline Mode=BSpline
    Stretch=true
    Up Axis=(X=0,Y=0,Z=1)
    debug=false
    Bones=@ItemArray_2.Value
    Controls=@ItemArray_3.Value
    ExecuteContext=@n11.B
   )
   > RigUnit_SetTransform = &RigUnit_SetTransform(
    Item=(Type=Bone,Name="dyn_photo_front_08")
    Space=LocalSpace
    Value=(Rotation=(X=-0.0238,Y=0,Z=0,W=0.9997),Translation=(X=6.5842,Y=0.2022,Z=-0.8679),Scale3D=(X=1,Y=1,Z=1))
    Weight=1
    bInitial=False
   )
   > &RigUnit_SetTransform(
    Item=(Type=Bone,Name="dyn_photo_front_09")
    Space=LocalSpace
    Value=(Rotation=(X=0,Y=0,Z=0,W=1),Translation=(X=2.8544,Y=0,Z=0),Scale3D=(X=1,Y=1,Z=1))
    Weight=1
    bInitial=False
   )
   ItemArray_5 = RigVMDispatchNode(Value="((Type=Bone,Name=\"ProcMove_photo_clavicle_back\"),(Type=Bone,Name=\"ProcMove_photo_back_01\"),(Type=Bone,Name=\"ProcMove_Photo_Back_02\"),(Type=Bone,Name=\"ProcMove_photo_back_03\"),(Type=Bone,Name=\"ProcMove_photo_back_06\"),(Type=Bone,Name=\"ProcMove_photo_back_09 ...(truncated)") {template="DISPATCH_RigVMDispatch_Constant(io Value)"}
   ItemArray_6 = RigVMDispatchNode(Value="((Type=Bone,Name=\"dyn_photo_clavicle_back\"),(Type=Bone,Name=\"dyn_photo_back_01\"),(Type=Bone,Name=\"dyn_photo_back_02\"),(Type=Bone,Name=\"dyn_photo_back_03\"),(Type=Bone,Name=\"dyn_photo_back_04\"),(Type=Bone,Name=\"dyn_photo_back_05\"),(Type=Bone,Name=\"dyn_photo_ ...(truncated)") {template="DISPATCH_RigVMDispatch_Constant(io Value)"}
   > SplineIK_1 = RigVMFunctionReferenceNode(
    Primary Axis=(X=-1,Y=0,Z=0)
    Samples Per Segment=16
    Secondary Spline Direction=(X=0,Y=0,Z=6)
    Spline Mode=BSpline
    Stretch=true
    Up Axis=(X=0,Y=0,Z=1)
    debug=false
    Controls=@ItemArray_5.Value
    Bones=@ItemArray_6.Value
   )
   n12 = &RigVMFunction_MathQuaternionToEuler(
    RotationOrder=XYZ
    Value=&RigUnit_GetRelativeTransformForItem(Child=(Type=Bone,Name="spine_05"), Parent=(Type=Bone,Name="pelvis"), bChildInitial=false, bParentInitial=false).RelativeTransform.Rotation
   )
   n13 = RigVMDispatchNode(Value=&RigVMFunction_MathDoubleRemap(SourceMaximum=20, SourceMinimum=-15, TargetMaximum=1, TargetMinimum=0, Value=@n12.Result.X).Result) {template="DISPATCH_RigVMDispatch_Constant(io Value)"}
   n14 = RigVMDispatchNode(Condition=true, False=false, True=true) {template="DISPATCH_RigVMDispatch_If(in Condition,in True,in False,out Result)"}
   n15 = RigVMDispatchNode(Elements=(TranslationFilter=(bX=True,bY=True,bZ=True),RotationFilter=(bX=false,bY=false,bZ=false),ScaleFilter=(bX=False,bY=False,bZ=False))) {template="DISPATCH_RigVMDispatch_MakeStruct(in Elements,out Struct)"}
   > &RigUnit_ParentConstraint(
    AdvancedSettings=(InterpolationType=Average,RotationOrderForFilter=XZY)
    Child=(Type=Bone,Name="ProcMove_photo_clavicle_back")
    Parents="((Item=(Type=Bone,Name=\"spine_03\"),Weight=0.0001),(Item=(Type=Bone,Name=\"spine_04\"),Weight=0.0157),(Item=(Type=Bone,Name=\"spine_05\"),Weight=0.1437),(Item=(Type=Bone,Name=\"clavicle_l\"),Weight=0.8364),(Item=(Type=Bone,Name=\"upperarm_twist_01_l\"),Weight=0.004 ...(truncated)"
    Weight=@n13.Value
    bMaintainOffset=@n14.Result
    Filter=@n15.Struct
   )
   > &RigUnit_ParentConstraint(
    AdvancedSettings=(InterpolationType=Average,RotationOrderForFilter=XZY)
    Child=(Type=Bone,Name="dyn_photo_back_01")
    Parents=((Item=(Type=Bone,Name="spine_03"),Weight=0.0024),(Item=(Type=Bone,Name="spine_04"),Weight=0.1602),(Item=(Type=Bone,Name="spine_05"),Weight=0.5228),(Item=(Type=Bone,Name="clavicle_l"),Weight=0.3151))
    Weight=@n13.Value
    bMaintainOffset=@n14.Result
    Filter=@n15.Struct
   )
   > &RigUnit_ParentConstraint(
    AdvancedSettings=(InterpolationType=Average,RotationOrderForFilter=XZY)
    Child=(Type=Bone,Name="dyn_photo_back_02")
    Parents=((Item=(Type=Bone,Name="spine_03"),Weight=0.0194),(Item=(Type=Bone,Name="spine_04"),Weight=0.5005),(Item=(Type=Bone,Name="spine_05"),Weight=0.4791),(Item=(Type=Bone,Name="clavicle_l"),Weight=0.001),(Item=(Type=Bone,Name="clavicle_r"),Weight=0))
    Weight=@n13.Value
    bMaintainOffset=@n14.Result
    Filter=@n15.Struct
   )
   > &RigUnit_ParentConstraint(
    AdvancedSettings=(InterpolationType=Average,RotationOrderForFilter=XZY)
    Child=(Type=Bone,Name="dyn_photo_back_03")
    Parents=((Item=(Type=Bone,Name="spine_01"),Weight=0.0683),(Item=(Type=Bone,Name="spine_02"),Weight=0.0939),(Item=(Type=Bone,Name="spine_03"),Weight=0.3443),(Item=(Type=Bone,Name="spine_04"),Weight=0.4681),(Item=(Type=Bone,Name="spine_05"),Weight=0.0253))
    Weight=@n13.Value
    bMaintainOffset=@n14.Result
    Filter=@n15.Struct
   )
   > &RigUnit_ParentConstraint(
    AdvancedSettings=(InterpolationType=Average,RotationOrderForFilter=XZY)
    Child=(Type=Bone,Name="dyn_photo_back_04")
    Parents=((Item=(Type=Bone,Name="pelvis"),Weight=0.0617),(Item=(Type=Bone,Name="spine_01"),Weight=0.7101),(Item=(Type=Bone,Name="spine_02"),Weight=0.1147),(Item=(Type=Bone,Name="spine_03"),Weight=0.0901),(Item=(Type=Bone,Name="spine_04"),Weight=0.0233))
    Weight=@n13.Value
    bMaintainOffset=@n14.Result
    Filter=@n15.Struct
   )
   > &RigUnit_ParentConstraint(
    AdvancedSettings=(InterpolationType=Average,RotationOrderForFilter=XZY)
    Child=(Type=Bone,Name="dyn_photo_back_05")
    Parents=((Item=(Type=Bone,Name="pelvis"),Weight=0.1774),(Item=(Type=Bone,Name="spine_01"),Weight=0.8221))
    Weight=@n13.Value
    bMaintainOffset=@n14.Result
    Filter=@n15.Struct
   )
   > &RigUnit_ModifyTransforms(
    ItemToModify="((Item=(Type=Bone,Name=\"dyn_photo_back_02\"),Transform=(Rotation=(X=0,Y=0.0132,Z=0,W=0.9999),Translation=(X=-0.0083,Y=0.1941,Z=0.3581),Scale3D=(X=1,Y=1,Z=1))),(Item=(Type=Bone,Name=\"dyn_photo_back_01\"),Transform=(Rotation=(X=0,Y=0,Z=0,W=1),Translation=(X=0. ...(truncated)"
    WeightMinimum=0
    Weight=@n13.Value
   )
   > &RigUnit_ModifyTransforms(
    ItemToModify="((Item=(Type=Bone,Name=\"dyn_photo_back_01\"),Transform=(Rotation=(X=-0.0568,Y=-0.0018,Z=0.032,W=0.9979),Translation=(X=1.9675,Y=1.3755,Z=0),Scale3D=(X=1,Y=1,Z=1))),(Item=(Type=Bone,Name=\"dyn_photo_back_02\"),Transform=(Rotation=(X=-0.023,Y=-0.0009,Z=0.0385,W ...(truncated)"
    WeightMinimum=0
    Weight=&RigVMFunction_MathDoubleRemap(SourceMaximum=60, SourceMinimum=30, TargetMaximum=1, TargetMinimum=0, Value=@n12.Result.X).Result
   )
   n16 = &RigUnit_GetTransform(Item=(Type=Bone,Name="upperarm_l"), bInitial=False)
   n17 = &RigUnit_GetTransform(Item=(Type=Bone,Name="spine_04"), bInitial=False)
   > &RigUnit_ModifyTransforms(
    ItemToModify="((Item=(Type=Bone,Name=\"dyn_photo_clavicle_top\"),Transform=(Rotation=(X=0,Y=0,Z=0.0447,W=-0.999),Translation=(X=-0.0322,Y=0.9809,Z=0),Scale3D=(X=1,Y=1,Z=1))),(Item=(Type=Bone,Name=\"dyn_photo_clavicle_back\"),Transform=(Rotation=(X=0,Y=0,Z=-0.0028,W=-1),Tran ...(truncated)"
    WeightMinimum=0
    Weight=&RigVMFunction_MathDoubleRemap(
     SourceMaximum=16.2
     SourceMinimum=17.52
     TargetMaximum=1
     TargetMinimum=0
     bClamp=False
     Value=&RigVMFunction_MathVectorDistance(A=@n16.Transform.Translation, B=@n17.Transform.Translation).Result
    ).Result
   )
   n18 = RigVMDispatchNode(Elements=(TranslationFilter=(bX=true,bY=true,bZ=true),RotationFilter=(bX=true,bY=true,bZ=true),ScaleFilter=(bX=False,bY=False,bZ=False))) {template="DISPATCH_RigVMDispatch_MakeStruct(in Elements,out Struct)"}
   n19 = &RigVMFunction_MathQuaternionToEuler(
    RotationOrder=XYZ
    Value=&RigUnit_GetRelativeTransformForItem(Child=(Type=Bone,Name="spine_05"), Parent=(Type=Bone,Name="pelvis"), bChildInitial=false, bParentInitial=false).RelativeTransform.Rotation
   )
   Maximum = &RigVMFunction_MathDoubleMax(
    A=&RigVMFunction_MathDoubleRemap(SourceMaximum=-22, SourceMinimum=0, TargetMaximum=1, TargetMinimum=0, Value=@n19.Result.Z).Result
    B=&RigVMFunction_MathDoubleRemap(SourceMaximum=-50, SourceMinimum=0, TargetMaximum=1, TargetMinimum=0, Value=@n19.Result.X).Result
   )
   n20 = RigVMDispatchNode(Value=@Maximum.Result) {template="DISPATCH_RigVMDispatch_Constant(io Value)"}
   > &RigUnit_ParentConstraint(
    AdvancedSettings=(InterpolationType=Average,RotationOrderForFilter=XZY)
    Child=(Type=Bone,Name="dyn_photo_front_01")
    Parents="((Item=(Type=Bone,Name=\"spine_03\"),Weight=0.0129),(Item=(Type=Bone,Name=\"spine_04\"),Weight=0.4538),(Item=(Type=Bone,Name=\"spine_05\"),Weight=0.2135),(Item=(Type=Bone,Name=\"clavicle_l\"),Weight=0.3087),(Item=(Type=Bone,Name=\"upperarm_twist_01_l\"),Weight=0.008 ...(truncated)"
    Filter=@n18.Struct
    Weight=&RigVMFunction_MathDoubleMul(B=0.5, A=@n20.Value).Result
   )
   > &RigUnit_ParentConstraint(
    AdvancedSettings=(InterpolationType=Average,RotationOrderForFilter=XZY)
    Child=(Type=Bone,Name="dyn_photo_front_02")
    Parents=((Item=(Type=Bone,Name="spine_03"),Weight=0.0114),(Item=(Type=Bone,Name="spine_04"),Weight=0.9695),(Item=(Type=Bone,Name="spine_05"),Weight=0.0191))
    Filter=@n18.Struct
    Weight=&RigVMFunction_MathDoubleMul(B=0.5, A=@n20.Value).Result
   )
   Multiply = &RigVMFunction_MathDoubleMul(B=0.4, A=@n20.Value)
   > &RigUnit_ParentConstraint(
    AdvancedSettings=(InterpolationType=Average,RotationOrderForFilter=XZY)
    Child=(Type=Bone,Name="dyn_photo_front_03")
    Parents=((Item=(Type=Bone,Name="spine_01"),Weight=0.0114),(Item=(Type=Bone,Name="spine_02"),Weight=0.9695),(Item=(Type=Bone,Name="spine_03"),Weight=0.0191),(Item=(Type=Bone,Name="spine_04"),Weight=0.9193),(Item=(Type=Bone,Name="spine_05"),Weight=0.0003))
    Filter=@n18.Struct
    Weight=@Multiply.Result
   )
   n21 = &RigVMFunction_MathDoubleMul(B=0.3, A=@n20.Value)
   > &RigUnit_ParentConstraint(
    AdvancedSettings=(InterpolationType=Average,RotationOrderForFilter=XZY)
    Child=(Type=Bone,Name="dyn_photo_front_04")
    Parents=((Item=(Type=Bone,Name="pelvis"),Weight=0.0116),(Item=(Type=Bone,Name="spine_01"),Weight=0.2568),(Item=(Type=Bone,Name="spine_02"),Weight=0.0489),(Item=(Type=Bone,Name="spine_03"),Weight=0.1479),(Item=(Type=Bone,Name="spine_04"),Weight=0.5347))
    Weight=@n21.Result
   )
   > &RigUnit_ParentConstraint(
    AdvancedSettings=(InterpolationType=Average,RotationOrderForFilter=XZY)
    Child=(Type=Bone,Name="dyn_photo_front_05")
    Parents=((Item=(Type=Bone,Name="pelvis"),Weight=0.0669),(Item=(Type=Bone,Name="spine_01"),Weight=0.6803),(Item=(Type=Bone,Name="spine_02"),Weight=0.0311),(Item=(Type=Bone,Name="spine_03"),Weight=0.0625),(Item=(Type=Bone,Name="spine_04"),Weight=0.1593))
    Filter=@n18.Struct
    Weight=@n21.Result
   )
   > &RigUnit_ModifyTransforms(
    ItemToModify="((Item=(Type=Bone,Name=\"dyn_photo_front_01\"),Transform=(Rotation=(X=0,Y=0,Z=0,W=-1),Translation=(X=0,Y=0,Z=0.5409),Scale3D=(X=1,Y=1,Z=1))),(Item=(Type=Bone,Name=\"dyn_photo_clavicle_front\"),Transform=(Rotation=(X=0,Y=-0.0371,Z=0,W=0.9993),Translation=(X=0,Y ...(truncated)"
    WeightMinimum=0
    Weight=&RigVMFunction_MathDoubleRemap(
     SourceMaximum=7
     SourceMinimum=0
     TargetMaximum=1
     TargetMinimum=0
     bClamp=False
     Value=&RigUnit_GetTransform(Item=(Type=Bone,Name="deform_upperarm_fr_l"), Space=LocalSpace, bInitial=False).Transform.Translation.Y
    ).Result
   )
   n22 = &RigUnit_GetRelativeTransformForItem(Child=(Type=Bone,Name="Master_photo"), Parent=(Type=Bone,Name="ProcMove_ParentFoto"), bChildInitial=true, bParentInitial=true)
   > /Script/ControlRig.RigUnit_SetRelativeTransformForItem(
    Child=(Type=Bone,Name="Master_photo")
    Parent=(Type=Bone,Name="ProcMove_ParentFoto")
    Weight=1
    bParentInitial=false
    bPropagateToChildren=true
    Value=@n22.RelativeTransform
   ) {
    is=struct RigVMUnitNode
    template="Set Relative Transform::Execute(in Child,in Parent,in bParentInitial,in Value,in Weight,in bPropagateToChildren)"
   }
   n23 = &RigUnit_GetRelativeTransformForItem(Child=(Type=Bone,Name="dyn_photo_back_07"), Parent=(Type=Bone,Name="ProcMove_ParentFoto"), bChildInitial=true, bParentInitial=true)
   > &RigUnit_SetRelativeTransformForItem(
    Child=(Type=Bone,Name="dyn_photo_back_07")
    Parent=(Type=Bone,Name="ProcMove_ParentFoto")
    Weight=1
    bParentInitial=false
    bPropagateToChildren=true
    Value=@n23.RelativeTransform
   ) {template="Set Relative Transform::Execute(in Child,in Parent,in bParentInitial,in Value,in Weight,in bPropagateToChildren)"}
   n24 = &RigVMFunction_MathQuaternionToEuler(
    RotationOrder=XYZ
    Value=&RigUnit_GetRelativeTransformForItem(Child=(Type=Bone,Name="spine_05"), Parent=(Type=Bone,Name="pelvis"), bChildInitial=false, bParentInitial=false).RelativeTransform.Rotation
   )
   &RigUnit_ModifyTransforms(
    ItemToModify="((Item=(Type=Bone,Name=\"ProcMove_photo_back_03\"),Transform=(Rotation=(X=0.0119,Y=0,Z=0,W=0.9999),Translation=(X=1.268,Y=0.0033,Z=2.4497),Scale3D=(X=1,Y=1,Z=1))),(Item=(Type=Bone,Name=\"ProcMove_photo_back_06\"),Transform=(Rotation=(X=0.1967,Y=0,Z=0,W=0.9805) ...(truncated)"
    WeightMinimum=0
    Weight=&RigVMFunction_MathDoubleRemap(SourceMaximum=60, SourceMinimum=0, TargetMaximum=1, TargetMinimum=0, bClamp=False, Value=@n24.Result.X).Result
    ExecutePin=@n11.A
   )
   > &RigUnit_ModifyTransforms(
    ItemToModify="((Item=(Type=Bone,Name=\"ProcMove_photo_back_03\"),Transform=(Rotation=(X=0,Y=0,Z=0,W=1),Translation=(X=-1,Y=0,Z=1),Scale3D=(X=1,Y=1,Z=1))),(Item=(Type=Bone,Name=\"ProcMove_Photo_Back_02\"),Transform=(Rotation=(X=0,Y=0,Z=0,W=1),Translation=(X=2.0432,Y=1.6729,Z ...(truncated)"
    Weight=1
    WeightMinimum=0
   )
   > &RigUnit_ModifyTransforms(
    ItemToModify="((Item=(Type=Bone,Name=\"ProcMove_photo_back_06\"),Transform=(Rotation=(X=0,Y=0,Z=0,W=1),Translation=(X=9.2581,Y=0,Z=1.1182),Scale3D=(X=1,Y=1,Z=1))),(Item=(Type=Bone,Name=\"ProcMove_photo_back_03\"),Transform=(Rotation=(X=0,Y=0,Z=0,W=1),Translation=(X=-9.5153, ...(truncated)"
    WeightMinimum=0
    Weight=&RigVMFunction_MathDoubleRemap(SourceMaximum=20, SourceMinimum=0, TargetMaximum=1, TargetMinimum=0, Value=@n24.Result.Z).Result
   )
   > &RigUnit_ModifyTransforms(
    ItemToModify="((Item=(Type=Bone,Name=\"ProcMove_photo_front_09\"),Transform=(Rotation=(X=0,Y=0,Z=0,W=1),Translation=(X=0,Y=0,Z=0.5),Scale3D=(X=1,Y=1,Z=1))),(Item=(Type=Bone,Name=\"ProceMove_photo_front_05\"),Transform=(Rotation=(X=0,Y=0,Z=0,W=1),Translation=(X=0.5,Y=0,Z=0.5 ...(truncated)"
    Weight=1
    WeightMinimum=0
   )
   n25 = &RigUnit_GetRelativeTransformForItem(Child=(Type=Bone,Name="clavicle_l"), Parent=(Type=Null,Name="clavicle_l_InitPose"), bChildInitial=False, bParentInitial=False)
   n26 = &RigVMFunction_MathQuaternionToEuler(RotationOrder=XYZ, Value=@n25.RelativeTransform.Rotation)
   > &RigUnit_ModifyTransforms(
    ItemToModify="((Item=(Type=Bone,Name=\"ProcMove_photo_front_02\"),Transform=(Rotation=(X=0,Y=0,Z=0,W=1),Translation=(X=-1.1173,Y=0,Z=1.3357),Scale3D=(X=1,Y=1,Z=1))),(Item=(Type=Bone,Name=\"ProceMove_photo_front_05\"),Transform=(Rotation=(X=0,Y=0,Z=0,W=1),Translation=(X=0,Y= ...(truncated)"
    WeightMinimum=0
    Weight=&RigVMFunction_MathDoubleRemap(SourceMaximum=25, SourceMinimum=0, TargetMaximum=1, TargetMinimum=0, Value=@n26.Result.X).Result
   )
   > &RigUnit_ModifyTransforms(
    ItemToModify="((Item=(Type=Bone,Name=\"ProcMove_photo_front_02\"),Transform=(Rotation=(X=0,Y=0,Z=0,W=1),Translation=(X=0.6411,Y=0,Z=0.3081),Scale3D=(X=1,Y=1,Z=1))),(Item=(Type=Bone,Name=\"ProceMove_photo_front_05\"),Transform=(Rotation=(X=0,Y=0,Z=0,W=1),Translation=(X=0,Y=0 ...(truncated)"
    WeightMinimum=0
    Weight=&RigVMFunction_MathDoubleRemap(SourceMaximum=-6, SourceMinimum=0, TargetMaximum=1, TargetMinimum=0, Value=@n26.Result.Z).Result
   )
   n27 = &RigUnit_GetTransform(Item=(Type=Bone,Name="NewBone"), bInitial=False)
   Add = &RigVMFunction_MathDoubleAdd(
    A=&RigVMFunction_MathVectorDistance(
     A=&RigUnit_GetTransform(Item=(Type=Bone,Name="thigh_r"), bInitial=False).Transform.Translation
     B=@n27.Transform.Translation
    ).Result
    B=&RigVMFunction_MathVectorDistance(
     B=&RigUnit_GetTransform(Item=(Type=Bone,Name="upperarm_l"), bInitial=False).Transform.Translation
     A=@n27.Transform.Translation
    ).Result
   )
   > &RigUnit_ModifyTransforms(
    ItemToModify="((Item=(Type=Bone,Name=\"ProcMove_photo_front_02\"),Transform=(Rotation=(X=0,Y=0,Z=0,W=1),Translation=(X=0,Y=0,Z=0.8269),Scale3D=(X=1,Y=1,Z=1))),(Item=(Type=Bone,Name=\"ProceMove_photo_front_05\"),Transform=(Rotation=(X=0,Y=0,Z=0,W=1),Translation=(X=-2.4101,Y= ...(truncated)"
    WeightMinimum=0
    Weight=&RigVMFunction_MathDoubleRemap(SourceMaximum=62, SourceMinimum=56, TargetMaximum=1, TargetMinimum=0, bClamp=False, Value=@Add.Result).Result
   )
   n28 = &RigVMFunction_MathQuaternionToEuler(
    RotationOrder=XYZ
    Value=&RigUnit_GetRelativeTransformForItem(Child=(Type=Bone,Name="spine_05"), Parent=(Type=Bone,Name="pelvis"), bChildInitial=false, bParentInitial=false).RelativeTransform.Rotation
   )
   > &RigUnit_ModifyTransforms(
    ItemToModify="((Item=(Type=Bone,Name=\"ProcMove_photo_front_02\"),Transform=(Rotation=(X=0,Y=0,Z=0,W=1),Translation=(X=-7.5289,Y=0,Z=7.0194),Scale3D=(X=1,Y=1,Z=1))),(Item=(Type=Bone,Name=\"ProceMove_photo_front_05\"),Transform=(Rotation=(X=0,Y=0,Z=0,W=-1),Translation=(X=6.3 ...(truncated)"
    WeightMinimum=0
    Weight=&RigVMFunction_MathDoubleRemap(SourceMaximum=-65, SourceMinimum=-10, TargetMaximum=1, TargetMinimum=0, Value=@n28.Result.X).Result
   )
   > &RigUnit_ModifyTransforms(
    ItemToModify="((Item=(Type=Bone,Name=\"ProcMove_photo_front_02\"),Transform=(Rotation=(X=-0.4923,Y=0,Z=0,W=0.8704),Translation=(X=12.0396,Y=8.1181,Z=-3.8454),Scale3D=(X=1,Y=1,Z=1))),(Item=(Type=Bone,Name=\"ProceMove_photo_front_05\"),Transform=(Rotation=(X=0,Y=0,Z=0,W=-1),T ...(truncated)"
    WeightMinimum=0
    Weight=&RigVMFunction_MathDoubleRemap(SourceMaximum=-105, SourceMinimum=-65, TargetMaximum=1, TargetMinimum=0, Value=@n28.Result.X).Result
   )
   n29 = &RigVMFunction_MathQuaternionToEuler(
    RotationOrder=XYZ
    Value=&RigUnit_GetTransform(Item=(Type=Bone,Name="spine_05"), Space=LocalSpace, bInitial=false).Transform.Rotation
   )
   > &RigUnit_ModifyTransforms(
    ItemToModify="((Item=(Type=Bone,Name=\"ProcMove_photo_front_02\"),Transform=(Rotation=(X=0,Y=0,Z=0,W=1),Translation=(X=0,Y=0,Z=0),Scale3D=(X=1,Y=1,Z=1))),(Item=(Type=Bone,Name=\"ProceMove_photo_front_05\"),Transform=(Rotation=(X=0,Y=0,Z=0,W=1),Translation=(X=-6.9942,Y=0,Z=0 ...(truncated)"
    WeightMinimum=0
    Weight=&RigVMFunction_MathDoubleRemap(SourceMaximum=-20, SourceMinimum=0, TargetMaximum=1, TargetMinimum=0, bClamp=False, Value=@n29.Result.X).Result
   )
   > &RigUnit_ModifyTransforms(
    ItemToModify="((Item=(Type=Bone,Name=\"ProcMove_photo_front_02\"),Transform=(Rotation=(X=0,Y=0,Z=0,W=1),Translation=(X=0,Y=0,Z=0),Scale3D=(X=1,Y=1,Z=1))),(Item=(Type=Bone,Name=\"ProceMove_photo_front_05\"),Transform=(Rotation=(X=0,Y=0,Z=0,W=1),Translation=(X=0,Y=0,Z=4.6542) ...(truncated)"
    WeightMinimum=0
    Weight=&RigVMFunction_MathDoubleRemap(SourceMaximum=-30, SourceMinimum=-10, TargetMaximum=1, TargetMinimum=0, bClamp=False, Value=@n29.Result.Z).Result
   )
   > n30 = &RigUnit_SphericalPoseReader(
    ActiveRegionScaleFactors=(PositiveWidth=3,NegativeWidth=0.5,PositiveHeight=3,NegativeHeight=3)
    ActiveRegionSize=0.07
    Debug=(bDrawDebug=false,bDraw2D=False,bDrawLocalAxes=False,DebugScale=25,DebugSegments=20,DebugThickness=0.25)
    DriverAxis=(X=1,Y=0,Z=0)
    DriverItem=(Type=Bone,Name="thigh_r")
    FalloffRegionScaleFactors=(PositiveWidth=1,NegativeWidth=1.2,PositiveHeight=1,NegativeHeight=1)
    FalloffSize=0.35
    FlipHeightScaling=False
    FlipWidthScaling=False
    OptionalParentItem=(Type=Bone,Name="None")
    RotationOffset=(X=-102.9608,Y=-168.9433,Z=14.0274)
   )
   > &RigUnit_ModifyTransforms(
    ItemToModify=((Item=(Type=Bone,Name="ProcMove_ParentFoto"),Transform=(Rotation=(X=0.212,Y=0.0391,Z=-0.049,W=0.9753),Translation=(X=-0.4849,Y=0.8123,Z=3.4125),Scale3D=(X=1,Y=1,Z=1))))
    WeightMinimum=0
    Weight=@n30.OutputParam
   )
   > n31 = &RigUnit_SphericalPoseReader(
    ActiveRegionScaleFactors=(PositiveWidth=3,NegativeWidth=0.5,PositiveHeight=1,NegativeHeight=3)
    ActiveRegionSize=0.07
    Debug=(bDrawDebug=false,bDraw2D=False,bDrawLocalAxes=False,DebugScale=25,DebugSegments=20,DebugThickness=0.25)
    DriverAxis=(X=1,Y=0,Z=0)
    DriverItem=(Type=Bone,Name="thigh_r")
    FalloffRegionScaleFactors=(PositiveWidth=1,NegativeWidth=1.5,PositiveHeight=0.9,NegativeHeight=1)
    FalloffSize=0.35
    FlipHeightScaling=False
    FlipWidthScaling=False
    OptionalParentItem=(Type=Bone,Name="None")
    RotationOffset=(X=-157.0822,Y=-167.8322,Z=-18.8617)
   )
   > &RigUnit_ModifyTransforms(
    ItemToModify="((Item=(Type=Bone,Name=\"ProcMove_ParentFoto\"),Transform=(Rotation=(X=0,Y=0.5508,Z=0,W=0.8346),Translation=(X=5.0697,Y=2.9787,Z=3.5959),Scale3D=(X=1,Y=1,Z=1))),(Item=(Type=Bone,Name=\"ProceMove_photo_front_05\"),Transform=(Rotation=(X=0,Y=0,Z=0,W=1),Translati ...(truncated)"
    WeightMinimum=0
    Weight=@n31.OutputParam
   )
   > n32 = &RigUnit_SphericalPoseReader(
    ActiveRegionScaleFactors=(PositiveWidth=1,NegativeWidth=1,PositiveHeight=1,NegativeHeight=1)
    ActiveRegionSize=0.1
    Debug=(bDrawDebug=True,bDraw2D=False,bDrawLocalAxes=False,DebugScale=25,DebugSegments=20,DebugThickness=0.25)
    DriverAxis=(X=1,Y=0,Z=0)
    DriverItem=(Type=Bone,Name="thigh_r")
    FalloffRegionScaleFactors=(PositiveWidth=1,NegativeWidth=1,PositiveHeight=1,NegativeHeight=1)
    FalloffSize=0.2
    FlipHeightScaling=False
    FlipWidthScaling=False
    OptionalParentItem=(Type=Bone,Name="None")
    RotationOffset=(X=15.1606,Y=-60.1698,Z=0)
   )
   > &RigUnit_ModifyTransforms(
    ItemToModify=((Item=(Type=Bone,Name="ProcMove_ParentFoto"),Transform=(Rotation=(X=0,Y=0.0593,Z=0,W=0.9982),Translation=(X=0,Y=0,Z=0),Scale3D=(X=1,Y=1,Z=1))))
    WeightMinimum=0
    Weight=@n32.OutputParam
   )
   n33 = &RigUnit_SphericalPoseReader(
    ActiveRegionScaleFactors=(PositiveWidth=3,NegativeWidth=1,PositiveHeight=3,NegativeHeight=1)
    ActiveRegionSize=0.07
    Debug=(bDrawDebug=true,bDraw2D=false,bDrawLocalAxes=False,DebugScale=25,DebugSegments=20,DebugThickness=0.25)
    DriverAxis=(X=-1,Y=0,Z=0)
    DriverItem=(Type=Bone,Name="thigh_l")
    ExecutePin=()
    FalloffRegionScaleFactors=(PositiveWidth=3,NegativeWidth=1.5,PositiveHeight=3,NegativeHeight=1)
    FalloffSize=0.3
    FlipHeightScaling=False
    FlipWidthScaling=False
    OptionalParentItem=(Type=Bone,Name="None")
    RotationOffset=(X=130,Y=0,Z=30)
   )
   > &RigUnit_ModifyTransforms(
    ItemToModify=((Item=(Type=Bone,Name="master_Art"),Transform=(Rotation=(X=0.293,Y=0.0981,Z=0.0511,W=0.9497),Translation=(X=-1,Y=-2,Z=-1),Scale3D=(X=1,Y=1,Z=1))))
    WeightMinimum=0
    Weight=&RigVMFunction_MathDoublePow(B=2, A=@n33.OutputParam) {template="Power::Execute(in A,in B,out Result)"}.Result
   )
   &RigUnit_ModifyTransforms(
    ExecutePin=()
    ItemToModify="((Item=(Type=Bone,Name=\"ProcMove_photo_front_02\"),Transform=(Rotation=(X=0,Y=0,Z=0,W=1),Translation=(X=0.6411,Y=0,Z=0.3081),Scale3D=(X=1,Y=1,Z=1))),(Item=(Type=Bone,Name=\"ProceMove_photo_front_05\"),Transform=(Rotation=(X=0,Y=0,Z=0,W=1),Translation=(X=0,Y=0 ...(truncated)"
    Weight=1
    WeightMinimum=0
   )
   &RigVMFunction_MathDoubleRemap(SourceMaximum=-65, SourceMinimum=-30, TargetMaximum=1, TargetMinimum=0, Value=0)
  ]}
]
hierarchy [
 root {loc=<0 0 0>, rot=<0 0 0 1>}
 pelvis {parent=root, loc=<0 -0.5301 98.0571>, rot=<0.0127 0.707 -0.0127 -0.707>}
 spine_01 {parent=pelvis, loc=<6.1208 -0.9328 0>, rot=<0 0 0.034 -0.9994>}
 spine_02 {parent=spine_01, loc=<6.5997 -0.2095 0>, rot=<0 0 0.0238 -0.9997>}
 spine_03 {parent=spine_02, loc=<6.4914 -0.1237 0>, rot=<0 0 0.0149 -0.9999>}
 spine_04 {parent=spine_03, loc=<6.9598 0.0794 0>, rot=<0 0 -0.0132 -0.9999>}
 spine_05 {parent=spine_04, loc=<6.9987 0 0>, rot=<0 0 -0.0015 -1>}
 neck_01 {parent=spine_05, loc=<11.3138 0.1529 0>, rot=<0 0 -0.2003 -0.9797>}
 neck_02 {parent=neck_01, loc=<4.5627 0 0>, rot=<0 0 0.0504 -0.9987>}
 head {parent=neck_02, loc=<5.3914 0 0>, rot=<0 0 0.115 -0.9934>}
 clavicle_l {parent=spine_05, loc=<5.4831 -0.1053 -3.2946>, rot=<0.0465 -0.7289 -0.009 -0.6829>}
 upperarm_l {parent=clavicle_l, loc=<9.8658 -1.2136 -0.4105>, rot=<0.0164 -0.3245 -0.0269 -0.9454>}
 lowerarm_l {parent=upperarm_l, loc=<26.465 -0.8345 0.259>, rot=<0.0036 0.0161 -0.2298 -0.9731>}
 hand_l {parent=lowerarm_l, loc=<22.7355 -0.3203 -0.6105>, rot=<-0.6784 -0.0223 0.0735 -0.7307>}
 thumb_01_l {parent=hand_l, loc=<1.3118 -0.3191 -2.459>, rot=<0.6546 -0.4042 0.0664 -0.6354>}
 thumb_02_l {parent=thumb_01_l, loc=<5.9336 -0.0448 0.0401>, rot=<-0.0037 0.0213 0.298 -0.9543>}
 thumb_03_l {parent=thumb_02_l, loc=<2.2223 0.3682 -0.0907>, rot=<0.0118 -0.013 0.0066 -0.9998>}
 index_metacarpal_l {parent=hand_l, loc=<1.6535 0.0183 -2.96>, rot=<0.0158 -0.0124 -0.0173 -0.9996>}
 index_01_l {parent=index_metacarpal_l, loc=<8.3008 0.0357 -0.0524>, rot=<-0.0044 0.0485 0.0924 -0.9945>}
 index_02_l {parent=index_01_l, loc=<3.3587 -0.3004 0.0452>, rot=<-0.002 0.0027 0.3225 -0.9466>}
 index_03_l {parent=index_02_l, loc=<2.5578 -0.123 -0.0252>, rot=<-0.002 0.0067 0.2558 -0.9667>}
 middle_metacarpal_l {parent=hand_l, loc=<2.1856 -0.1725 -0.5984>, rot=<0.006 0.0168 -0.0329 -0.9993>}
 middle_01_l {parent=middle_metacarpal_l, loc=<7.3275 -0.2447 0.0129>, rot=<-0.0208 0.0042 0.0858 -0.9961>}
 middle_02_l {parent=middle_01_l, loc=<4.0123 0.1401 0.1175>, rot=<-0.0218 0.0258 0.3512 -0.9357>}
 middle_03_l {parent=middle_02_l, loc=<2.6077 0.0534 0.1373>, rot=<0.0082 0.0188 0.1708 -0.9851>}
 ring_metacarpal_l {parent=hand_l, loc=<2.2074 -0.3625 1.3638>, rot=<-0.0284 0.0722 -0.0458 -0.9959>}
 ring_01_l {parent=ring_metacarpal_l, loc=<7.0295 -0.6197 -0.1446>, rot=<-0.0489 -0.0252 0.1892 -0.9804>}
 ring_02_l {parent=ring_01_l, loc=<3.6174 0.4647 -0.1027>, rot=<0 0 0.2806 -0.9598>}
 ring_03_l {parent=ring_02_l, loc=<2.4152 0.114 0.0199>, rot=<0 0 0.1189 -0.9929>}
 pinky_metacarpal_l {parent=hand_l, loc=<1.7616 -0.9753 3.3814>, rot=<-0.0397 0.0814 -0.069 -0.9935>}
 pinky_01_l {parent=pinky_metacarpal_l, loc=<7.1709 -0.4831 -0.1094>, rot=<-0.0654 -0.0035 0.2242 -0.9723>}
 pinky_02_l {parent=pinky_01_l, loc=<2.8274 0.0893 -0.1043>, rot=<0.0037 0.0011 0.2854 -0.9584>}
 pinky_03_l {parent=pinky_02_l, loc=<2.2046 0.0307 -0.0078>, rot=<-0.0002 0.0098 0.1445 -0.9895>}
 deform_wrist_dn_base_l {parent=hand_l, loc=<0 -1.5326 0>, rot=<0 0 0 -1>}
 deform_wrist_dn_l {parent=deform_wrist_dn_base_l, loc=<0 0 0>, rot=<0 0 0 1>}
 deform_wrist_up_base_l {parent=hand_l, loc=<0 2.3009 0>, rot=<0 0 0 -1>}
 deform_wrist_up_l {parent=deform_wrist_up_base_l, loc=<0 0 0>, rot=<0 0 0 1>}
 lowerarm_twist_02_l {parent=lowerarm_l, loc=<10.6707 0 0>, rot=<0 0 0 1>}
 lowerarm_twist_01_l {parent=lowerarm_l, loc=<18.316 0 0>, rot=<0 0 0 1>}
 deform_elbow_in_base_l {parent=lowerarm_l, loc=<0.1834 3.6062 -0.2572>, rot=<0 0 0 -1>}
 deform_elbow_in_l {parent=deform_elbow_in_base_l, loc=<0 0 0>, rot=<0 0 0 1>}
 deform_elbow_out_base_l {parent=lowerarm_l, loc=<0 -4 0>, rot=<0 0 0 1>}
 deform_elbow_out_l {parent=deform_elbow_out_base_l, loc=<0 0 0>, rot=<0 0 0 1>}
 hand_l_InitPose {parent=lowerarm_l, type=Null, loc=<22.7355 -0.3203 -0.6105>, rot=<0.6784 0.0223 -0.0735 0.7307>}
 upperarm_twist_01_l {parent=upperarm_l, loc=<7.2135 -0.1134 0>, rot=<0 0 0 1>}
 deform_upperarm_bk_base_l {parent=upperarm_twist_01_l, loc=<-6.4053 -4.3746 0.314>, rot=<0 0 0 -1>}
 deform_upperarm_bk_l {parent=deform_upperarm_bk_base_l, loc=<0 0 0>, rot=<0 0 0 1>}
 deform_upperarm_fr_base_l {parent=upperarm_twist_01_l, loc=<-5.6312 5.092 0.3354>, rot=<0 0 0 -1>}
 deform_upperarm_fr_l {parent=deform_upperarm_fr_base_l, loc=<0 0 0>, rot=<0 0 0 1>}
 deform_upperarm_up_base_l {parent=upperarm_twist_01_l, loc=<-8.891 0.4825 3.9517>, rot=<0 0 0 -1>}
 deform_upperarm_up_l {parent=deform_upperarm_up_base_l, loc=<0 0 0>, rot=<0 0 0 1>}
 upperarm_twist_02_l {parent=upperarm_l, loc=<19.7998 -0.1134 0>, rot=<0 0 0 1>}
 dyn_photo_clavicle_top {parent=clavicle_l, loc=<7.1925 -0.2874 5.3558>, rot=<-0.0225 -0.1739 -0.7102 -0.6818>}
 dyn_photo_clavicle_back {parent=clavicle_l, loc=<6.9435 -4.5313 4.3448>, rot=<-0.301 0.1093 -0.6437 -0.6951>}
 dyn_photo_clavicle_front {parent=clavicle_l, loc=<6.4531 3.4495 3.4552>, rot=<0.1892 -0.3603 -0.6961 -0.5914>}
 ProcMove_photo_clavicle_front {parent=clavicle_l, loc=<6.3199 3.5739 3.0135>, rot=<0.1892 -0.3603 -0.6961 -0.5914>}
 ProcMove_photo_back_01 {parent=clavicle_l, loc=<8.2492 -6.7653 4.046>, rot=<0.5401 -0.2854 0.4638 0.6417>}
 ProcMove_photo_clavicle_back {parent=clavicle_l, loc=<6.9435 -4.5313 4.3448>, rot=<-0.301 0.1093 -0.6437 -0.6951>}
 pelvis_to_clavicle_l {parent=clavicle_l, type=Null, loc=<0 0 0>, rot=<0.0041 -0.7304 0.0306 0.6823>}
 ProcMove_Photo_Back_02 {parent=clavicle_l, loc=<4.54 -7.7888 -1.1189>, rot=<0.5769 -0.2576 0.4797 0.6088>}
 clavicle_r {parent=spine_05, loc=<5.4831 -0.1053 3.2946>, rot=<0.7289 0.0465 -0.6829 0.009>}
 upperarm_r {parent=clavicle_r, loc=<-9.8658 1.2136 0.4105>, rot=<0.0164 -0.3245 -0.0269 -0.9454>}
 lowerarm_r {parent=upperarm_r, loc=<-26.465 0.8345 -0.259>, rot=<0.0036 0.0161 -0.2298 -0.9731>}
 hand_r {parent=lowerarm_r, loc=<-22.7355 0.3203 0.6106>, rot=<-0.6784 -0.0223 0.0735 -0.7307>}
 thumb_01_r {parent=hand_r, loc=<-1.3118 0.3191 2.459>, rot=<0.6546 -0.4042 0.0664 -0.6354>}
 thumb_02_r {parent=thumb_01_r, loc=<-5.9336 0.0448 -0.0401>, rot=<-0.0037 0.0213 0.298 -0.9543>}
 thumb_03_r {parent=thumb_02_r, loc=<-2.2223 -0.3682 0.0907>, rot=<0.0118 -0.013 0.0066 -0.9998>}
 index_metacarpal_r {parent=hand_r, loc=<-1.6535 -0.0183 2.96>, rot=<0.0158 -0.0124 -0.0173 -0.9996>}
 index_01_r {parent=index_metacarpal_r, loc=<-8.3008 -0.0357 0.0524>, rot=<-0.0044 0.0485 0.0924 -0.9945>}
 index_02_r {parent=index_01_r, loc=<-3.3587 0.3004 -0.0452>, rot=<-0.002 0.0027 0.3225 -0.9466>}
 index_03_r {parent=index_02_r, loc=<-2.5578 0.123 0.0252>, rot=<-0.002 0.0067 0.2558 -0.9667>}
 middle_metacarpal_r {parent=hand_r, loc=<-2.1856 0.1725 0.5984>, rot=<0.006 0.0168 -0.0329 -0.9993>}
 middle_01_r {parent=middle_metacarpal_r, loc=<-7.3275 0.2447 -0.0129>, rot=<-0.0208 0.0042 0.0858 -0.9961>}
 middle_02_r {parent=middle_01_r, loc=<-4.0123 -0.1401 -0.1175>, rot=<-0.0218 0.0258 0.3512 -0.9357>}
 middle_03_r {parent=middle_02_r, loc=<-2.6077 -0.0534 -0.1373>, rot=<0.0082 0.0188 0.1708 -0.9851>}
 ring_metacarpal_r {parent=hand_r, loc=<-2.2074 0.3625 -1.3638>, rot=<-0.0284 0.0722 -0.0458 -0.9959>}
 ring_01_r {parent=ring_metacarpal_r, loc=<-7.0295 0.6197 0.1446>, rot=<-0.0489 -0.0252 0.1892 -0.9804>}
 ring_02_r {parent=ring_01_r, loc=<-3.6174 -0.4647 0.1027>, rot=<0 0 0.2806 -0.9598>}
 ring_03_r {parent=ring_02_r, loc=<-2.4152 -0.114 -0.0199>, rot=<0 0 0.1189 -0.9929>}
 pinky_metacarpal_r {parent=hand_r, loc=<-1.7616 0.9753 -3.3814>, rot=<-0.0397 0.0814 -0.069 -0.9935>}
 pinky_01_r {parent=pinky_metacarpal_r, loc=<-7.1709 0.4831 0.1094>, rot=<-0.0654 -0.0035 0.2242 -0.9723>}
 pinky_02_r {parent=pinky_01_r, loc=<-2.8274 -0.0893 0.1043>, rot=<0.0037 0.0011 0.2854 -0.9584>}
 pinky_03_r {parent=pinky_02_r, loc=<-2.2046 -0.0307 0.0078>, rot=<-0.0002 0.0098 0.1445 -0.9895>}
 deform_wrist_dn_base_r {parent=hand_r, loc=<-0.0001 1.5328 0>, rot=<0 0 0 -1>}
 deform_wrist_dn_r {parent=deform_wrist_dn_base_r, loc=<0 0 0>, rot=<0 0 0 1>}
 deform_wrist_up_base_r {parent=hand_r, loc=<-0.0002 -2.3007 -0.0001>, rot=<0 0 0 -1>}
 deform_wrist_up_r {parent=deform_wrist_up_base_r, loc=<0 0 0>, rot=<0 0 0 1>}
 lowerarm_twist_02_r {parent=lowerarm_r, loc=<-10.6707 0 0>, rot=<0 0 0 -1>}
 lowerarm_twist_01_r {parent=lowerarm_r, loc=<-18.316 0 0>, rot=<0 0 0 -1>}
 deform_elbow_in_base_r {parent=lowerarm_r, loc=<-0.1834 -3.6062 0.2572>, rot=<0 0 0 -1>}
 deform_elbow_in_r {parent=deform_elbow_in_base_r, loc=<0 0 0>, rot=<0 0 0 1>}
 deform_elbow_out_base_r {parent=lowerarm_r, loc=<0 4 -0.0001>, rot=<0 0 0 -1>}
 deform_elbow_out_r {parent=deform_elbow_out_base_r, loc=<0 0 0>, rot=<0 0 0 1>}
 upperarm_twist_01_r {parent=upperarm_r, loc=<-7.2135 0.1134 0>, rot=<0 0 0 -1>}
 deform_upperarm_bk_base_r {parent=upperarm_twist_01_r, loc=<6.4053 4.3746 -0.314>, rot=<0 0 0 -1>}
 deform_upperarm_bk_r {parent=deform_upperarm_bk_base_r, loc=<0 0 0>, rot=<0 0 0 1>}
 deform_upperarm_fr_base_r {parent=upperarm_twist_01_r, loc=<5.6312 -5.092 -0.3354>, rot=<0 0 0 -1>}
 deform_upperarm_fr_r {parent=deform_upperarm_fr_base_r, loc=<0 0 0>, rot=<0 0 0 1>}
 deform_upperarm_up_base_r {parent=upperarm_twist_01_r, loc=<8.891 -0.4825 -3.9517>, rot=<0 0 0 -1>}
 deform_upperarm_up_r {parent=deform_upperarm_up_base_r, loc=<0 0 0>, rot=<0 0 0 1>}
 upperarm_twist_02_r {parent=upperarm_r, loc=<-19.7998 0.1134 0>, rot=<0 0 0 -1>}
 attach_backpack {parent=spine_05, loc=<-0.123 -13.473 0>, rot=<-0.0093 -0.707 0.0093 -0.707>}
 Backpack_BR {parent=attach_backpack, type=Null, loc=<0 0 0>, rot=<0 0 0 1>}
 attach_cape {parent=spine_05, loc=<9.017 -11.714 0>, rot=<0.0616 -0.7044 -0.0616 -0.7044>}
 cape {parent=attach_cape, type=Null, loc=<0 0 0>, rot=<0 0 0 1>}
 dyn_photo_back_01 {parent=spine_05, loc=<4.708 -8.1967 -7.3513>, rot=<-0.7177 -0.2953 0.0755 -0.6261>}
 dyn_photo_front_01 {parent=spine_05, loc=<2.48 7.8407 -8.1659>, rot=<-0.312 -0.714 -0.6211 -0.0849>}
 dyn_photo_front_02 {parent=spine_05, loc=<-2.7642 11.1415 -6.0375>, rot=<-0.2488 -0.7049 -0.6604 0.0709>}
 dyn_photo_front_03 {parent=spine_05, loc=<-5.9996 12.0199 -4.4358>, rot=<-0.1666 -0.6931 -0.6834 0.1572>}
 dyn_photo_front_04 {parent=spine_05, loc=<-9.8717 12.0293 -2.4244>, rot=<-0.1299 -0.7058 -0.6644 0.2086>}
 dyn_photo_back_02 {parent=spine_05, loc=<-1.4455 -9.4119 -4.1757>, rot=<-0.7129 -0.1985 0.1222 -0.6614>}
 ProcMove_photo_front_02 {parent=spine_05, loc=<-1.4008 12.6986 -6.7204>, rot=<0.3265 0.7083 0.6257 0.0145>}
 clavicle_l_InitPose {parent=spine_05, type=Null, loc=<5.483 -0.1053 -3.2946>, rot=<0.0465 -0.7289 -0.009 -0.6829>}
 ProcMove_photo_back_03 {parent=spine_04, loc=<0.5658 -11.5042 -0.4456>, rot=<0.7394 0.1049 -0.2133 0.63>}
 dyn_photo_back_03 {parent=spine_03, loc=<4.0808 -8.6853 0.5394>, rot=<-0.6911 -0.1272 0.1975 -0.6835>}
 dyn_photo_front_05 {parent=spine_03, loc=<-2.5997 11.2172 1.2711>, rot=<-0.1487 -0.6795 -0.6788 0.2355>}
 ProceMove_photo_front_05 {parent=spine_03, loc=<5.8897 15.5603 -4.1011>, rot=<-0.1487 -0.6795 -0.6788 0.2355>}
 dyn_photo_back_03_null {parent=spine_03, type=Null, loc=<4.88 -8.8782 1.6317>, rot=<-0.0387 0.706 0.0387 0.706>}
 NewBone {parent=spine_03, loc=<0 11.8218 0>, rot=<0 0 0 1>}
 dyn_photo_back_04 {parent=spine_02, loc=<5.0217 -7.6874 3.5206>, rot=<-0.5703 -0.1498 0.2304 -0.7741>}
 dyn_photo_back_05 {parent=spine_01, loc=<6.6371 -6.7145 6.8819>, rot=<-0.4452 -0.1932 0.2666 -0.8327>}
 dyn_photo_front_06 {parent=spine_01, loc=<6.4597 10.2394 4.3457>, rot=<0.1742 0.6192 0.7355 -0.2129>}
 thigh_l {parent=pelvis, loc=<-1.0855 -0.2109 -8.7053>, rot=<0.0009 0.0355 0.0155 -0.9993>}
 calf_l {parent=thigh_l, loc=<-42.5284 -0.0019 -0.7297>, rot=<0.0006 -0.0161 -0.0391 -0.9991>}
 foot_l {parent=calf_l, loc=<-46.7401 -2.5014 -0.592>, rot=<0.0427 -0.0238 0.0353 -0.9982>}
 ball_l {parent=foot_l, loc=<-3.5658 10.8066 -0.1544>, rot=<0.0001 -0.0001 -0.7063 -0.7079>}
 calf_twist_02_l {parent=calf_l, loc=<-19.1546 -1.7943 -0.2638>, rot=<0.0237 0.0095 -0.0035 -0.9997>}
 dyn_laces_in_l {parent=calf_twist_02_l, loc=<-10.3561 4.2005 0.0368>, rot=<-0.7332 -0.0271 0.0473 -0.6778>}
 dyn_laces_out_l {parent=calf_twist_02_l, loc=<-10.2463 4.2319 -0.6589>, rot=<-0.648 -0.0035 0.0215 -0.7613>}
 calf_twist_01_l {parent=calf_l, loc=<-35.8483 -1.9197 -0.5794>, rot=<0.0237 0.0095 -0.0035 -0.9997>}
 deform_knee_in_base_l {parent=calf_l, loc=<0.8031 -5.0231 0.2456>, rot=<0 0 0 -1>}
 deform_knee_in_l {parent=deform_knee_in_base_l, loc=<0 0 0>, rot=<0 0 0 1>}
 deform_knee_out_base_l {parent=calf_l, loc=<0 4 0>, rot=<0 0 0 -1>}
 deform_knee_out_l {parent=deform_knee_out_base_l, loc=<0 0 0>, rot=<0 0 0 1>}
 thigh_twist_01_l {parent=thigh_l, loc=<-27.952 0 0>, rot=<0.0013 0 0 -1>}
 deform_groin_base_l {parent=thigh_l, loc=<0.4801 5.6776 -0.2486>, rot=<0 0 0 -1>}
 deform_groin_l {parent=deform_groin_base_l, loc=<0 0 0>, rot=<0 0 0 1>}
 thigh_r {parent=pelvis, loc=<-1.0855 -0.2109 8.7053>, rot=<0.0355 -0.0009 0.9993 0.0155>}
 calf_r {parent=thigh_r, loc=<42.5284 0.0019 0.7297>, rot=<0.0006 -0.0161 -0.0391 -0.9991>}
 foot_r {parent=calf_r, loc=<46.7401 2.5014 0.592>, rot=<0.0427 -0.0238 0.0353 -0.9982>}
 ball_r {parent=foot_r, loc=<3.5658 -10.8066 0.1544>, rot=<0.0001 -0.0001 -0.7063 -0.7079>}
 calf_twist_02_r {parent=calf_r, loc=<19.1546 1.7943 0.2638>, rot=<0.0237 0.0095 -0.0035 -0.9997>}
 dyn_laces_in_r {parent=calf_twist_02_r, loc=<10.3561 -4.2005 -0.0368>, rot=<-0.7332 -0.0271 0.0473 -0.6778>}
 dyn_laces_out_r {parent=calf_twist_02_r, loc=<10.2463 -4.2319 0.6589>, rot=<-0.648 -0.0035 0.0215 -0.7613>}
 calf_twist_01_r {parent=calf_r, loc=<35.8483 1.9197 0.5794>, rot=<0.0237 0.0095 -0.0035 -0.9997>}
 deform_knee_in_base_r {parent=calf_r, loc=<-0.8031 5.0231 -0.2457>, rot=<0 0 0 -1>}
 deform_knee_in_r {parent=deform_knee_in_base_r, loc=<0 0 0>, rot=<0 0 0 1>}
 deform_knee_out_base_r {parent=calf_r, loc=<0 -4 0>, rot=<0 0 0 -1>}
 deform_knee_out_r {parent=deform_knee_out_base_r, loc=<0 0 0>, rot=<0 0 0 1>}
 thigh_twist_01_r {parent=thigh_r, loc=<27.952 0 0>, rot=<0.0013 0 0 -1>}
 deform_groin_base_r {parent=thigh_r, loc=<-0.4801 -5.6776 0.2486>, rot=<0 0 0 -1>}
 deform_groin_r {parent=deform_groin_base_r, loc=<0 0 0>, rot=<0 0 0 1>}
 deform_glute_base_l {parent=pelvis, loc=<-7.3466 -9.0213 -9.1245>, rot=<0.0172 0.031 -0.4477 -0.8935>}
 deform_glute_l {parent=deform_glute_base_l, loc=<0 0 0>, rot=<0 0 0 1>}
 deform_glute_base_r {parent=pelvis, loc=<-7.3464 -9.0213 9.1245>, rot=<0.031 -0.0172 0.8935 -0.4477>}
 deform_glute_r {parent=deform_glute_base_r, loc=<0 0 0>, rot=<0 0 0 1>}
 bn_belt01 {parent=pelvis, loc=<4.5974 1.678 14.4867>, rot=<-0.0738 -0.3453 -0.7098 -0.6095>}
 dyn_pumpkin {parent=bn_belt01, loc=<0.02 3.1046 -0.0534>, rot=<0.1269 0.1252 -0.7233 -0.6671>}
 bn_belt02 {parent=pelvis, loc=<3.2183 5.4126 12.2849>, rot=<0.1643 -0.4827 -0.6619 -0.5495>}
 bn_belt03 {parent=pelvis, loc=<1.6176 7.8427 7.0106>, rot=<0.381 -0.4785 -0.6067 -0.5077>}
 bn_belt04 {parent=pelvis, loc=<0.0157 9.1222 -2.2662>, rot=<0.5082 -0.5493 -0.5054 -0.4296>}
 bn_belt05 {parent=pelvis, loc=<-0.1425 7.0242 -10.7592>, rot=<0.6182 -0.6335 -0.2811 -0.3708>}
 bn_belt06 {parent=pelvis, loc=<0.228 2.4088 -14.7383>, rot=<0.7013 -0.6819 -0.0331 -0.2052>}
 bn_belt07 {parent=pelvis, loc=<0.8051 -4.4576 -14.7863>, rot=<0.7241 -0.6587 0.2045 -0.0006>}
 bn_belt08 {parent=pelvis, loc=<1.5325 -8.209 -12.6649>, rot=<0.7682 -0.5167 0.348 0.1474>}
 bn_belt09 {parent=pelvis, loc=<2.731 -10.8481 -7.8664>, rot=<0.7208 -0.3896 0.4964 0.2868>}
 bn_belt10 {parent=pelvis, loc=<4.0706 -11.4212 -0.1706>, rot=<0.6295 -0.3163 0.585 0.4017>}
 bn_belt11 {parent=pelvis, loc=<4.9145 -9.3489 8.27>, rot=<-0.5026 0.1681 -0.6691 -0.5211>}
 bn_belt12 {parent=pelvis, loc=<5.0922 -4.8016 12.6452>, rot=<-0.3493 -0.0469 -0.6948 -0.6269>}
 master_Art {parent=pelvis, loc=<0.1946 -2.3572 -15.8277>, rot=<0.1674 -0.9797 0.0385 -0.1035>}
 dyn_Art {parent=master_Art, loc=<0 0 0>, rot=<0 0 0 1>}
 dyn_easel {parent=dyn_Art, loc=<12.7061 10.4686 5.8541>, rot=<0 0 0 1>}
 dyn_brushA {parent=dyn_Art, loc=<11.7035 11.2977 3.539>, rot=<-0.0987 -0.7196 0.683 -0.0768>}
 dyn_brushB {parent=dyn_Art, loc=<12.8157 11.5857 3.035>, rot=<-0.0428 -0.7571 0.6514 -0.0266>}
 dyn_brushC {parent=dyn_Art, loc=<12.8681 11.2749 2.2909>, rot=<-0.0013 -0.7896 0.6123 0.0402>}
 master_Pack {parent=pelvis, loc=<5.6348 -8.4295 10.3464>, rot=<0.1583 -0.9319 0.3086 0.1058>}
 dyn_pack {parent=master_Pack, loc=<0 0 0>, rot=<0 0 0 1>}
 dyn_toy {parent=dyn_pack, loc=<-0.4024 -1.7977 -4.0626>, rot=<-0.019 -0.472 -0.0354 -0.8807>}
 dyn_ringToy {parent=dyn_toy, loc=<1.552 0 0>, rot=<0 0.1271 0 -0.9919>}
 dyn_toyHead {parent=dyn_ringToy, loc=<2.2032 0 0.0718>, rot=<0 0.3445 0 -0.9388>}
 dyn_toyBody {parent=dyn_toyHead, loc=<5.844 0 0.0348>, rot=<0 0 0 1>}
 dyn_ToyTail {parent=dyn_toyBody, loc=<3.6162 0 0>, rot=<0.0569 0.0272 0.0289 -0.9976>}
 dyn_ToyTailMid {parent=dyn_ToyTail, loc=<8.7079 -1.1391 0.6189>, rot=<0.0125 0.0537 0.0849 -0.9949>}
 dyn_ToyTailEnd {parent=dyn_ToyTailMid, loc=<5.3571 -0.5098 0.1253>, rot=<0.001 0.0208 0.0989 -0.9949>}
 dyn_photo_front_07 {parent=pelvis, loc=<8.6566 8.3879 7.3818>, rot=<0.2045 0.5269 0.7967 -0.2142>}
 dyn_photo_front_08 {parent=dyn_photo_front_07, loc=<6.5842 0.2022 -0.8679>, rot=<0.0238 0 0 -0.9997>}
 dyn_photo_front_09 {parent=dyn_photo_front_08, loc=<2.8544 0 0>, rot=<0 0 0 1>}
 Master_photo {parent=dyn_photo_front_09, loc=<8.8531 -0.2815 -1.715>, rot=<0.5643 0.5435 0.168 -0.5983>}
 dyn_photo {parent=Master_photo, loc=<0 0 0>, rot=<0 0 0 1>}
 dyn_photo_back_06 {parent=pelvis, loc=<7.3177 -6.4297 11.125>, rot=<-0.2958 -0.283 0.2378 -0.8809>}
 dyn_photo_back_07 {parent=dyn_photo_back_06, loc=<-5.5645 -0.1516 -0.492>, rot=<0.1111 0.0897 -0.0101 -0.9897>}
 dyn_photo_back_08 {parent=dyn_photo_back_07, loc=<-5.919 -0.6244 -0.7048>, rot=<0.3251 0.0462 -0.0873 -0.9405>}
 dyn_photo_back_09 {parent=dyn_photo_back_08, loc=<-2.789 0 0>, rot=<0 0 0 1>}
 ProcMove_ParentFoto {parent=pelvis, loc=<6.8302 0.9917 10.7575>, rot=<-0.1104 0.2337 0.0116 0.9659>}
 ProcMove_photo_front_09 {parent=ProcMove_ParentFoto, loc=<-6.4299 5.3316 0.3709>, rot=<0.0039 -0.4477 -0.8889 0.0972>}
 ProcMove_photo_back_06 {parent=ProcMove_ParentFoto, loc=<-2.0236 -5.6074 0.7634>, rot=<0.4595 0.0401 -0.1675 0.8713>}
 ProcMove_photo_back_09 {parent=ProcMove_ParentFoto, loc=<-12.3403 -3.9198 1.3453>, rot=<0 0 -0.1342 0.9909>}
]
