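# Physics asset for the art pack on the F_MED_CrispSeason female player cosmetic, defining jointed bodies for its secondary motion.
# A box body for the pack hangs off the spherical master_Art body with small swing and twist limits.
# Three capsule paintbrushes and a spherical easel body are constrained to the pack with tighter limits.
# The easel joint locks twist and one swing axis, leaving a narrow single-axis sway.
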PhysicsAsset /BRCosmetics/Characters/Player/Female/Medium/Bodies/F_MED_CrispSeason/Meshes/F_MED_CrispSeason_Physics_ArtPack
bodies {
 dyn_Art [{shapeType=Box, extentX=19.6252, extentY=7.5567, extentZ=24.016, shapeName=dyn_Art_box, center=<4.9919 0 2.9912>, rotation=<rot 0 0 90>}]
 dyn_brushA [{shapeType=Capsule, length=7.3055, radius=0.6152, shapeName=dyn_brushB_capsule, center=<18.8197 0 0>, rotation=<rot -90 0 0>}]
 dyn_brushB [{shapeType=Capsule, length=7.5015, radius=0.6213, shapeName=dyn_brushB_capsule, center=<19.4995 0 0>, rotation=<rot -90 0 0>}]
 dyn_brushC [{shapeType=Capsule, length=7.3055, radius=0.6152, shapeName=dyn_brushB_capsule, center=<19.723 0 0>, rotation=<rot -90 0 0>}]
 dyn_easel [{shapeType=Sphere, radius=2.4747, shapeName=dyn_easel_sphere}]
 master_Art [{shapeType=Sphere, radius=3, shapeName=master_Art_sphere}]
}
constraints [
 {bone1Name=dyn_Art, bone2Name=master_Art, swing1LimitDegrees=8, swing2LimitDegrees=8, twistLimitDegrees=8}
 {
  bone1Name=dyn_easel
  bone2Name=dyn_Art
  swing1LimitDegrees=5
  swing2LimitDegrees=0
  swing2Motion=Locked
  twistLimitDegrees=0
  twistMotion=Locked
 }
 {bone1Name=dyn_brushC, bone2Name=dyn_Art, swing1LimitDegrees=3, swing2LimitDegrees=0, twistLimitDegrees=2}
 {bone1Name=dyn_brushA, bone2Name=dyn_Art, swing1LimitDegrees=3, swing2LimitDegrees=0, twistLimitDegrees=2}
 {bone1Name=dyn_brushB, bone2Name=dyn_Art, swing1LimitDegrees=3, swing2LimitDegrees=0, twistLimitDegrees=2}
]
