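# Physics asset for the CrispSeason medium female player body cosmetic, covering its dangling accessories and legs.
# Tightly limited chains hang shoelaces from calf twist bones, a toy with tail from the pack, and a pumpkin from the belt.
# The photo prop joins front and back chains, whose inner links mostly lock to one swing axis.
# Thigh and calf shapes form the leg bodies, linked by free-rotation constraints.
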PhysicsAsset /BRCosmetics/Characters/Player/Female/Medium/Bodies/F_MED_CrispSeason/Meshes/F_MED_CrispSeason_Physics_FE
bodies {
 bn_belt01 [{shapeType=Sphere, radius=1.9708, shapeName=bn_belt01_sphere, center=<-0.2909 -0.1921 -0.7258>}]
 calf_l [{
   shapeType=Capsule
   length=11.5633
   radius=3.9875
   shapeName=calf_l_capsule
   center=<-35.4187 -1.5435 -0.7861>
   rotation=<rot 89.0404 -3.1857 -178.7332>
  }]
 calf_r [
  {
   shapeType=Capsule
   length=11.5633
   radius=3.9875
   shapeName=calf_r_capsule
   center=<35.4187 1.5435 0.7861>
   rotation=<rot 89.0404 -3.1857 1.2668>
  }
  {
   shapeType=Capsule
   length=38.6134
   radius=6.9746
   shapeName=calf_r_capsule
   center=<24.2655 2.2823 0.4119>
   rotation=<rot 88.4388 179.9999 180>
  }
 ]
 calf_twist_02_l [{shapeType=Sphere, radius=2.7017, shapeName=calf_twist_02_l_sphere}]
 calf_twist_02_r [{shapeType=Sphere, radius=2.7017, shapeName=calf_twist_02_r_sphere}]
 dyn_ToyTail [{shapeType=Sphere, radius=4.4944, shapeName=dyn_ToyTail_sphere, center=<5.6811 -0.9238 0.3091>}]
 dyn_ToyTailEnd [{shapeType=Sphere, radius=2.9336, shapeName=dyn_ToyTailEnd_sphere, center=<3.0601 -0.2829 -0.4012>}]
 dyn_ToyTailMid [{shapeType=Sphere, radius=5.0205, shapeName=dyn_ToyTailMid_sphere, center=<1.7794 -0.1943 0>}]
 dyn_laces_in_l [{shapeType=Sphere, radius=1.7029, shapeName=calf_twist_02_l_sphere, center=<-6.5324 0 -1.5314>}]
 dyn_laces_in_r [{shapeType=Sphere, radius=1.7029, shapeName=calf_twist_02_r_sphere, center=<6.5324 0 1.5314>}]
 dyn_laces_out_l [{shapeType=Sphere, radius=1.7651, shapeName=calf_twist_02_l_sphere, center=<-6.6573 -2.0287 -1.7469>}]
 dyn_laces_out_r [{shapeType=Sphere, radius=1.7651, shapeName=calf_twist_02_r_sphere, center=<6.6573 2.0287 1.7469>}]
 dyn_pack [{shapeType=Box, extentX=14.1153, extentY=4.9035, extentZ=8.8364, shapeName=dyn_pack_box, center=<4.1765 -1.865 0>}]
 dyn_photo [{shapeType=Box, extentX=4.7414, extentY=12.5229, extentZ=8.8185, shapeName=dyn_photo_box, center=<1.6009 0 0>}]
 dyn_photo_back_06 [{
   shapeType=Capsule
   length=7.6934
   radius=1.3536
   shapeName=dyn_toy_capsule
   center=<-4.6797 -0.0002 -1.3889>
   rotation=<rot 83.1613 -179.797 -179.7664>
  }]
 dyn_photo_back_07 [{
   shapeType=Capsule
   length=4.4891
   radius=1.6966
   shapeName=dyn_toy_capsule
   center=<-1.8727 -0.673 -0.9544>
   rotation=<rot 85.0313 -0.3498 -3.8648>
  }]
 dyn_photo_back_08 [{shapeType=Sphere, radius=1.3684, shapeName=dyn_photo_back_08_sphere}]
 dyn_photo_back_09 [{shapeType=Sphere, radius=1.3495, shapeName=dyn_photo_back_09_sphere}]
 dyn_photo_front_06 [{shapeType=Sphere, radius=1, shapeName=dyn_photo_front_07_sphere}]
 dyn_photo_front_07 [{shapeType=Sphere, radius=1, shapeName=dyn_photo_front_07_sphere}]
 dyn_photo_front_08 [{shapeType=Sphere, radius=1, shapeName=dyn_photo_front_08_sphere}]
 dyn_photo_front_09 [{shapeType=Sphere, radius=1.3312, shapeName=dyn_photo_front_09_sphere, center=<0.7786 -0.0914 0.1286>}]
 dyn_pumpkin [{shapeType=Sphere, radius=2.8487, shapeName=dyn_pumpkin_sphere, center=<3.03 0.1028 0.0377>}]
 dyn_ringToy [{shapeType=Capsule, length=0.9584, radius=0.5, shapeName=dyn_toy_capsule, center=<1.9375 0.0682 -0.0218>, rotation=<rot 0 1.2441 -88.8178>}]
 dyn_toy [{shapeType=Capsule, length=0.9584, radius=0.5, shapeName=dyn_toy_capsule, center=<1.519 0.0591 -0.0218>, rotation=<rot 0 1.2441 -88.8178>}]
 dyn_toyBody [{shapeType=Sphere, radius=2.0299, shapeName=dyn_toyBody_sphere, center=<1.8726 0.1137 0.0149>}]
 dyn_toyHead [{shapeType=Sphere, radius=3.5899, shapeName=dyn_toyHead_sphere, center=<3.1928 0.0551 -0.1294>}]
 thigh_l [{
   shapeType=Capsule
   length=23.8966
   radius=8.1345
   shapeName=thigh_l_capsule
   center=<-13.9127 0 1.0395>
   rotation=<rot 84.6699 179.9997 -0.0001>
  }]
 thigh_r [
  {shapeType=Sphere, radius=8.3901, shapeName=thigh_r_sphere, center=<4.4127 2.1858 0>}
  {
   shapeType=Capsule
   length=23.4198
   radius=8.6112
   shapeName=thigh_r_capsule
   center=<13.9127 0 -1.0395>
   rotation=<rot 87.4315 179.9997 179.9998>
  }
 ]
}
constraints [
 {bone1Name=dyn_laces_in_r, bone2Name=calf_twist_02_r, swing1LimitDegrees=30, swing2LimitDegrees=30, twistLimitDegrees=0}
 {bone1Name=dyn_laces_out_r, bone2Name=calf_twist_02_r, swing1LimitDegrees=30, swing2LimitDegrees=30, twistLimitDegrees=0}
 {
  bone1Name=calf_twist_02_r
  bone2Name=calf_r
  swing1LimitDegrees=45
  swing1Motion=Free
  swing2LimitDegrees=45
  swing2Motion=Free
  twistLimitDegrees=45
  twistMotion=Free
 }
 {
  bone1Name=calf_twist_02_l
  bone2Name=calf_l
  swing1LimitDegrees=45
  swing1Motion=Free
  swing2LimitDegrees=45
  swing2Motion=Free
  twistLimitDegrees=45
  twistMotion=Free
 }
 {bone1Name=dyn_laces_in_l, bone2Name=calf_twist_02_l, swing1LimitDegrees=30, swing2LimitDegrees=30, twistLimitDegrees=0}
 {bone1Name=dyn_laces_out_l, bone2Name=calf_twist_02_l, swing1LimitDegrees=30, swing2LimitDegrees=30, twistLimitDegrees=0}
 {bone1Name=dyn_toy, bone2Name=dyn_pack, swing1LimitDegrees=30, swing2LimitDegrees=30, twistLimitDegrees=0}
 {bone1Name=dyn_ringToy, bone2Name=dyn_toy, swing1LimitDegrees=0, swing2LimitDegrees=30, twistLimitDegrees=0}
 {bone1Name=dyn_toyHead, bone2Name=dyn_ringToy, swing1LimitDegrees=15, swing2LimitDegrees=30, twistLimitDegrees=0}
 {bone1Name=dyn_toyBody, bone2Name=dyn_toyHead, swing1LimitDegrees=10, swing2LimitDegrees=10, twistLimitDegrees=0}
 {bone1Name=dyn_ToyTail, bone2Name=dyn_toyBody, swing1LimitDegrees=25, swing2LimitDegrees=25, twistLimitDegrees=0}
 {bone1Name=dyn_ToyTailMid, bone2Name=dyn_ToyTail, swing1LimitDegrees=25, swing2LimitDegrees=25, twistLimitDegrees=0}
 {bone1Name=dyn_ToyTailEnd, bone2Name=dyn_ToyTailMid, swing1LimitDegrees=25, swing2LimitDegrees=25, twistLimitDegrees=0}
 {
  bone1Name=calf_r
  bone2Name=thigh_r
  swing1LimitDegrees=45
  swing1Motion=Free
  swing2LimitDegrees=45
  swing2Motion=Free
  twistLimitDegrees=45
  twistMotion=Free
 }
 {bone1Name=dyn_pumpkin, bone2Name=bn_belt01, swing1LimitDegrees=15, swing2LimitDegrees=15, twistLimitDegrees=5}
 {bone1Name=dyn_photo_front_09, bone2Name=dyn_photo_front_08, swing1LimitDegrees=15, swing2LimitDegrees=15, twistLimitDegrees=0}
 {
  bone1Name=dyn_photo_front_08
  bone2Name=dyn_photo_front_07
  swing1LimitDegrees=15
  swing1Motion=Locked
  swing2LimitDegrees=15
  twistLimitDegrees=0
  twistMotion=Locked
 }
 {bone1Name=dyn_photo_back_09, bone2Name=dyn_photo_back_08, swing1LimitDegrees=15, swing2LimitDegrees=15, twistLimitDegrees=0}
 {bone1Name=dyn_photo_back_07, bone2Name=dyn_photo_back_06, swing1LimitDegrees=15, swing2LimitDegrees=15, twistLimitDegrees=0}
 {
  bone1Name=dyn_photo_back_08
  bone2Name=dyn_photo_back_07
  swing1LimitDegrees=15
  swing1Motion=Locked
  swing2LimitDegrees=15
  twistLimitDegrees=0
  twistMotion=Locked
 }
 {bone1Name=dyn_photo, bone2Name=dyn_photo_front_09, swing1LimitDegrees=15, swing2LimitDegrees=15, twistLimitDegrees=15}
 {bone1Name=dyn_photo, bone2Name=dyn_photo_back_09, swing1LimitDegrees=15, swing2LimitDegrees=15, twistLimitDegrees=15}
 {
  bone1Name=dyn_photo_front_07
  bone2Name=dyn_photo_front_06
  swing1LimitDegrees=15
  swing1Motion=Locked
  swing2LimitDegrees=15
  twistLimitDegrees=0
  twistMotion=Locked
 }
 {
  bone1Name=calf_l
  bone2Name=thigh_l
  swing1LimitDegrees=45
  swing1Motion=Free
  swing2LimitDegrees=45
  swing2Motion=Free
  twistLimitDegrees=45
  twistMotion=Free
 }
]
