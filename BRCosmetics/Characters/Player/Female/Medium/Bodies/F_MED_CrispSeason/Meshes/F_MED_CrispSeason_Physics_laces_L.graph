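# Physics asset for the left-side laces on the CrispSeason female medium player body, supporting secondary lace motion.
# A calf_l capsule represents the lower leg, with inner and outer lace bones as spheres.
# Each lace body is constrained directly to the calf with matching narrow swing limit angles and a zero twist angle.
# A small sphere on dyn_photo_front_06 has no constraint in this asset.
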PhysicsAsset /BRCosmetics/Characters/Player/Female/Medium/Bodies/F_MED_CrispSeason/Meshes/F_MED_CrispSeason_Physics_laces_L
bodies {
 calf_l [{
   shapeType=Capsule
   length=11.5633
   radius=3.9875
   shapeName=calf_l_capsule
   center=<-35.4187 -1.5435 -0.7861>
   rotation=<rot 89.0404 -3.1857 -178.7332>
  }]
 dyn_laces_in_l [{shapeType=Sphere, radius=1.7029, shapeName=calf_twist_02_l_sphere, center=<-6.5324 0 -1.5314>}]
 dyn_laces_out_l [{shapeType=Sphere, radius=1.7651, shapeName=calf_twist_02_l_sphere, center=<-6.6573 -2.0287 -1.7469>}]
 dyn_photo_front_06 [{shapeType=Sphere, radius=1, shapeName=dyn_photo_front_07_sphere}]
}
constraints [
 {bone1Name=dyn_laces_in_l, bone2Name=calf_l, swing1LimitDegrees=30, swing2LimitDegrees=30, twistLimitDegrees=0}
 {bone1Name=dyn_laces_out_l, bone2Name=calf_l, swing1LimitDegrees=30, swing2LimitDegrees=30, twistLimitDegrees=0}
]
