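# Post-process-style AnimBP for the F_MED_CrispSeason female character body, adding secondary physics motion to the cosmetic mesh.
# It copies the parent mesh pose, runs a Control Rig, then caches it for two alternative pelvis-space RigidBody sims joined by a bool blend.
# Chained RigidBody nodes then simulate the art pack and both sets of laces.
# An interpolated upward force, stronger while skydiving, feeds one pelvis sim during jumps or falls without an open parachute.
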AnimBlueprint /BRCosmetics/Characters/Player/Female/Medium/Bodies/F_MED_CrispSeason/Meshes/F_MED_CrispSeason_AnimBP
target_skeleton=&/F_MED_CrispSeason_Skeleton
graphs [
 AnimGraph {nodes [
   RigidBody() {
    bClampLinearTranslationLimitToRefPose=true
    overridePhysicsAsset=&/F_MED_CrispSeason_Physics
    simulationSpace=BaseBoneSpace
    baseBoneRef {boneName=pelvis}
    componentAppliedLinearAccClamp=<300 300 300>
    overrideWorldGravity {z=-980}
    simSpaceSettings {dampingAlpha=0, maxAngularAcceleration=20, maxAngularVelocity=10, maxLinearAcceleration=20, maxLinearVelocity=10, worldAlpha=1}
   }
   n1 = RigidBody(ComponentPose=Local To Component(LocalPose=Use cached pose 'PreAnim'())) {
    bClampLinearTranslationLimitToRefPose=true
    overridePhysicsAsset=&/F_MED_CrispSeason_Physics_FE
    simulationSpace=BaseBoneSpace
    baseBoneRef {boneName=pelvis}
    overrideWorldGravity {z=-500}
    simSpaceSettings {
     dampingAlpha=0
     maxAngularAcceleration=10
     maxAngularVelocity=10
     maxLinearAcceleration=10
     maxLinearVelocity=10
     velocityScaleZ=0
     worldAlpha=1
    }
   }
   n2 = Component To Local(ComponentPose=@n1)
   n3 = RigidBody(ComponentPose=Local To Component(LocalPose=Use cached pose 'PreAnim'()), ExternalForce=Get ExternalForce() -> (ExternalForce:Vector)) {
    bClampLinearTranslationLimitToRefPose=true
    overridePhysicsAsset=&F_MED_CrispSeason_Physics
    simulationSpace=BaseBoneSpace
    baseBoneRef {boneName=pelvis}
    overrideWorldGravity {z=-980}
    simSpaceSettings {dampingAlpha=0, maxAngularAcceleration=5, maxAngularVelocity=5, maxLinearAcceleration=150, maxLinearVelocity=5, worldAlpha=1}
   }
   n4 = Component To Local(ComponentPose=@n3)
   n5 = RigidBody(ComponentPose=Local To Component(LocalPose=Blend Poses by bool(BlendPose_0=@n2, BlendPose_1=@n4))) {
    overridePhysicsAsset=&/F_MED_CrispSeason_Physics_ArtPack
    simulationSpace=BaseBoneSpace
    baseBoneRef {boneName=master_Art}
    componentAppliedLinearAccClamp=<1500 1500 1500>
    overrideWorldGravity {z=-980}
    simSpaceSettings {maxAngularAcceleration=50, maxAngularVelocity=20, maxLinearAcceleration=50, maxLinearVelocity=20, worldAlpha=1}
   }
   n6 = RigidBody(ComponentPose=@n5) {
    bClampLinearTranslationLimitToRefPose=true
    overridePhysicsAsset=&/F_MED_CrispSeason_Physics_laces_R
    simulationSpace=BaseBoneSpace
    baseBoneRef {boneName=calf_r}
    overrideWorldGravity {z=-980}
    simSpaceSettings {dampingAlpha=0, maxAngularAcceleration=250, maxAngularVelocity=50, maxLinearAcceleration=250, maxLinearVelocity=50, worldAlpha=1}
   }
   n7 = RigidBody(ComponentPose=@n6) {
    bClampLinearTranslationLimitToRefPose=true
    overridePhysicsAsset=&/F_MED_CrispSeason_Physics_laces_L
    simulationSpace=BaseBoneSpace
    baseBoneRef {boneName=calf_l}
    overrideWorldGravity {z=-980}
    simSpaceSettings {dampingAlpha=0, maxAngularAcceleration=250, maxAngularVelocity=50, maxLinearAcceleration=250, maxLinearVelocity=50, worldAlpha=1}
   }
   Output Pose(Result=Component To Local(ComponentPose=@n7)) {layerGroup=None}
   n8 = Control Rig(Source=Copy Pose From Mesh() {bUseAttachedParent=true}) {controlRigAssetReference {blueprintRigClass=&/F_MED_CrispSeason_CtrlRig.F_MED_CrispSeason_CtrlRig_C}}
   Save cached pose 'PreAnim'(Pose=@n8) {cacheName=PreAnim}
  ]}
 EventGraph {nodes [
   n1 = Event BlueprintUpdateAnimation()
   n2 = SelectFloat(A=800, B=500, bPickA=Get bIsSkyDiving())
   n3 = AND Boolean(A=OR Boolean(A=Get bIsFalling(), B=Get bIsJumping()), B=NOT Boolean(A=Get bIsParachuteOpened()))
   > Set ExternalForce(ExternalForce:Vector=VInterpTo!(
     Current:Vector=Get ExternalForce()
     Target_X:double=0
     Target_Y:double=0
     Target_Z:double=SelectFloat(A=@n2, B=0, bPickA=@n3)
     DeltaTime:float=@n1.DeltaTimeX
     InterpSpeed:float=5
    ) -> (ReturnValue:Vector)) -> (Output_Get:Vector)
   SelectFloat(A=900, B=600, bPickA=false)
  ]}
]
variables [ExternalForce {type=Vector}]
