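# Control Rig for the face accessory of the CrispSeason female body, correcting the hair braid as head and left clavicle move.
# Head rotation relative to spine_05 is remapped to weight a braid_Root offset, then parent constraints tie dyn_braid_01 to spine_05 and dyn_braid_02 to spine_05 and clavicle_l.
# Clavicle rotation against an init-pose null weights further braid offsets.
# Constraints for dyn_braid_03 to 05 sit outside the execution chain.
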ControlRigBlueprint /BRCosmetics/Characters/Player/Female/Medium/Bodies/F_MED_CrispSeason/Meshes/Parts/F_MED_CrispSeason_FaceAcc_CtrlRig
graphs [
 RigVMFunctionLibrary {links [], nodes []}
 RigVMModel {nodes [
   &RigUnit_ParentConstraint(
    AdvancedSettings=(InterpolationType=Average,RotationOrderForFilter=XZY)
    Child=(Type=Bone,Name="dyn_braid_03")
    ExecutePin=()
    Filter=(TranslationFilter=(bX=True,bY=True,bZ=true),RotationFilter=(bX=True,bY=True,bZ=True),ScaleFilter=(bX=True,bY=True,bZ=True))
    Parents=((Item=(Type=Bone,Name="spine_05"),Weight=1))
    Weight=1
   )
   > &RigUnit_ParentConstraint(
    AdvancedSettings=(InterpolationType=Average,RotationOrderForFilter=XZY)
    Child=(Type=Bone,Name="dyn_braid_04")
    Filter=(TranslationFilter=(bX=True,bY=True,bZ=true),RotationFilter=(bX=True,bY=True,bZ=True),ScaleFilter=(bX=True,bY=True,bZ=True))
    Parents=((Item=(Type=Bone,Name="spine_05"),Weight=1))
    Weight=1
   )
   > &RigUnit_ParentConstraint(
    AdvancedSettings=(InterpolationType=Average,RotationOrderForFilter=XZY)
    Child=(Type=Bone,Name="dyn_braid_05")
    Filter=(TranslationFilter=(bX=True,bY=True,bZ=true),RotationFilter=(bX=True,bY=True,bZ=True),ScaleFilter=(bX=True,bY=True,bZ=True))
    Parents=((Item=(Type=Bone,Name="spine_05"),Weight=1))
    Weight=1
   )
   RigUnit_BeginExecution = &RigUnit_BeginExecution()
   RigUnit_GetRelativeTransformForItem = &RigUnit_GetRelativeTransformForItem(Child=(Type=Bone,Name="head"), Parent=(Type=Bone,Name="spine_05"), bChildInitial=False, bParentInitial=False)
   RigVMFunction_MathQuaternionToEuler = &RigVMFunction_MathQuaternionToEuler(RotationOrder=ZYX, Value=@RigUnit_GetRelativeTransformForItem.RelativeTransform.Rotation)
   Remap = &RigVMFunction_MathDoubleRemap(SourceMaximum=-60, SourceMinimum=-20, TargetMaximum=1, TargetMinimum=0, bClamp=False, Value=@RigVMFunction_MathQuaternionToEuler.Result.Z)
   &RigUnit_ModifyTransforms(
    ItemToModify=((Item=(Type=Bone,Name="braid_Root"),Transform=(Rotation=(X=0,Y=0,Z=0.1837,W=-0.983),Translation=(X=-4.5096,Y=-0.5663,Z=0),Scale3D=(X=1,Y=1,Z=1))))
    WeightMinimum=0
    Weight=@Remap.Result
    ExecutePin=@RigUnit_BeginExecution.ExecutePin
   )
   > &RigUnit_ParentConstraint(
    AdvancedSettings=(InterpolationType=Average,RotationOrderForFilter=XZY)
    Child=(Type=Bone,Name="dyn_braid_01")
    Filter=(TranslationFilter=(bX=True,bY=True,bZ=true),RotationFilter=(bX=True,bY=True,bZ=True),ScaleFilter=(bX=True,bY=True,bZ=True))
    Parents=((Item=(Type=Bone,Name="spine_05"),Weight=1))
    Weight=1
   )
   > &RigUnit_ParentConstraint(
    AdvancedSettings=(InterpolationType=Average,RotationOrderForFilter=XZY)
    Child=(Type=Bone,Name="dyn_braid_02")
    Filter=(TranslationFilter=(bX=True,bY=True,bZ=true),RotationFilter=(bX=True,bY=True,bZ=True),ScaleFilter=(bX=True,bY=True,bZ=True))
    Parents=((Item=(Type=Bone,Name="spine_05"),Weight=1),(Item=(Type=Bone,Name="clavicle_l"),Weight=1))
    Weight=1
   )
   n1 = &RigUnit_GetRelativeTransformForItem(Child=(Type=Bone,Name="clavicle_l"), Parent=(Type=Null,Name="clavicle_l_InitPose"), bChildInitial=False, bParentInitial=False)
   > &RigUnit_ModifyTransforms(
    ItemToModify="((Item=(Type=Bone,Name=\"dyn_braid_02\"),Transform=(Rotation=(X=0,Y=0,Z=-0.0403,W=0.9992),Translation=(X=0,Y=1.4006,Z=2.6214),Scale3D=(X=1,Y=1,Z=1))),(Item=(Type=Bone,Name=\"dyn_braid_03\"),Transform=(Rotation=(X=-0.0039,Y=0.0328,Z=-0.1195,W=-0.9923),Translati ...(truncated)"
    WeightMinimum=0
    Weight=&RigVMFunction_MathDoubleRemap(
     SourceMaximum=-45
     SourceMinimum=-20
     TargetMaximum=1
     TargetMinimum=0
     Value=&RigVMFunction_MathQuaternionToEuler(RotationOrder=XYZ, Value=@n1.RelativeTransform.Rotation).Result.Y
    ).Result
   )
  ]}
]
hierarchy [
 root {loc=<0 0 0>, rot=<0 0 0 1>}
 pelvis {parent=root, loc=<0 -0.5301 98.0571>, rot=<0.0127 0.707 -0.0127 -0.707>}
 spine_01 {parent=pelvis, loc=<6.1208 -0.9328 0>, rot=<0 0 0.034 -0.9994>}
 spine_02 {parent=spine_01, loc=<6.5997 -0.2095 0>, rot=<0 0 0.0238 -0.9997>}
 spine_03 {parent=spine_02, loc=<6.4914 -0.1237 0>, rot=<0 0 0.0149 -0.9999>}
 spine_04 {parent=spine_03, loc=<6.9598 0.0794 0>, rot=<0 0 -0.0132 -0.9999>}
 spine_05 {parent=spine_04, loc=<6.9987 0 0>, rot=<0 0 -0.0015 -1>}
 neck_01 {parent=spine_05, loc=<11.3138 0.1529 0>, rot=<0 0 -0.2003 -0.9797>}
 neck_02 {parent=neck_01, loc=<4.5627 0 0>, rot=<0 0 0.0504 -0.9987>}
 head {parent=neck_02, loc=<5.3914 0 0>, rot=<0 0 0.115 -0.9934>}
 FaceAcc_Root {parent=head, loc=<0 0 0>, rot=<0 0 0 1>}
 dyn_earring {parent=FaceAcc_Root, loc=<0.2106 -0.0199 7.1443>, rot=<-0.0012 -0.9318 -0.3629 0.0095>}
 dyn_ear_r {parent=FaceAcc_Root, loc=<12.1451 -3.0218 5.7266>, rot=<0.6241 0.1709 0.1335 -0.7507>}
 dyn_ear_l {parent=FaceAcc_Root, loc=<12.1452 -3.0218 -5.7266>, rot=<0.1709 -0.6241 0.7507 0.1335>}
 dyn_HairStrand_ear_r_01 {parent=FaceAcc_Root, loc=<4.9558 2.0683 6.8519>, rot=<0.0479 0.6811 0.7224 -0.1093>}
 dyn_HairStrand_ear_r_02 {parent=dyn_HairStrand_ear_r_01, loc=<4.8705 0.8632 -0.4566>, rot=<0.1927 -0.1077 -0.4758 -0.8514>}
 dyn_HairStrand_A_r {parent=FaceAcc_Root, loc=<5.6042 4.0062 7.3192>, rot=<0.2575 0.5205 0.5249 -0.6223>}
 dyn_HairStrand_B_r {parent=FaceAcc_Root, loc=<7.6978 5.7424 5.7308>, rot=<0.0157 -0.7569 -0.634 0.1577>}
 dyn_HairStrand_front_01 {parent=FaceAcc_Root, loc=<10.9952 7.4117 0.5342>, rot=<-0.3161 -0.9127 -0.2257 0.1266>}
 dyn_HairStrand_front_02 {parent=dyn_HairStrand_front_01, loc=<3.3181 0.1847 0.1022>, rot=<-0.0303 -0.1998 0.148 -0.9681>}
 dyn_HairStrand_front_03 {parent=dyn_HairStrand_front_02, loc=<3.908 0.4559 -2.7263>, rot=<-0.0814 -0.8371 -0.043 -0.5393>}
 dyn_HairStrand_front_04 {parent=dyn_HairStrand_front_02, loc=<3.1144 0.5361 -0.6887>, rot=<-0.1906 -0.6309 -0.2175 -0.72>}
 HairStrand_front {parent=FaceAcc_Root, loc=<10.7291 7.9949 -2.2566>, rot=<-0.182 -0.9667 0.1628 0.0769>}
 dyn_HairStrand_l_01 {parent=FaceAcc_Root, loc=<10.203 7.0411 -3.6612>, rot=<-0.1879 -0.9315 0.3094 -0.0363>}
 dyn_HairStrand_l_02 {parent=dyn_HairStrand_l_01, loc=<4.152 0.2923 0.3974>, rot=<0.1448 0.4444 -0.025 -0.8837>}
 dyn_HairStrand_l {parent=FaceAcc_Root, loc=<5.5579 3.5147 -6.9545>, rot=<0.1915 -0.8192 0.3208 -0.435>}
 dyn_HairStrand_ear_l_01 {parent=FaceAcc_Root, loc=<3.5796 0.7349 -8.133>, rot=<-0.0457 -0.6006 0.797 -0.0451>}
 dyn_HairStrand_ear_l_02 {parent=dyn_HairStrand_ear_l_01, loc=<3.2748 0.0114 0.6661>, rot=<0.0837 0.1892 -0.3609 -0.9094>}
 braid_Root {parent=FaceAcc_Root, loc=<0.9251 -8.9201 0>, rot=<-0.1382 -0.9904 0 0>}
 dyn_braid_01 {parent=braid_Root, loc=<5.8292 -0.2823 0>, rot=<0 0 0.1423 -0.9898>}
 dyn_braid_02 {parent=dyn_braid_01, loc=<0.8912 0.6839 4.9822>, rot=<-0.1952 0.2779 -0.5825 -0.7385>}
 dyn_braid_03 {parent=dyn_braid_02, loc=<5.688 0 -0.9554>, rot=<0.1409 -0.3103 0.1894 -0.9209>}
 dyn_braid_04 {parent=dyn_braid_03, loc=<5.4532 0 0>, rot=<0 0 0 1>}
 dyn_braid_05 {parent=dyn_braid_04, loc=<4.925 0 0>, rot=<-0.0612 -0.0701 0.0326 -0.9951>}
 dyn_braid_06 {parent=dyn_braid_05, loc=<5.3773 0 0>, rot=<-0.0848 -0.0154 0.144 -0.9858>}
 dyn_braid_07 {parent=dyn_braid_06, loc=<4.2756 -1.0108 0.9693>, rot=<0 -0.1532 0 -0.9882>}
 dyn_HairStrand2_l {parent=FaceAcc_Root, loc=<7.0738 5.6778 -6.4451>, rot=<0.1408 -0.9892 -0.0086 -0.0403>}
 clavicle_l {parent=spine_05, loc=<5.4831 -0.1053 -3.2946>, rot=<0.0465 -0.7289 -0.009 -0.6829>}
 upperarm_l {parent=clavicle_l, loc=<9.8658 -1.2136 -0.4105>, rot=<0.0164 -0.3245 -0.0269 -0.9454>}
 clavicle_l_InitPose {parent=spine_05, type=Null, loc=<5.483 -0.1053 -3.2946>, rot=<0.0465 -0.7289 -0.009 -0.6829>}
]
