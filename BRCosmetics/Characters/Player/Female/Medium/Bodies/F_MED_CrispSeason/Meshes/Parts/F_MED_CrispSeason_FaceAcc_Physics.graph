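# Physics asset for the F_MED_CrispSeason face accessory mesh, defining simulated motion for hair strands, ears and an earring.
# Each bone gets one sphere body around a large FaceAcc_Root sphere, with most strand bodies reusing the same sphere shape.
# Constraints anchor parts to the root or chain strand segments, with front_03 and front_04 branching from front_02.
# Tight angular limits keep motion subtle, while the earring swings wider on one axis with twist locked.
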PhysicsAsset /BRCosmetics/Characters/Player/Female/Medium/Bodies/F_MED_CrispSeason/Meshes/Parts/F_MED_CrispSeason_FaceAcc_Physics
bodies {
 FaceAcc_Root [{shapeType=Sphere, radius=9.1482, shapeName=FaceAcc_Root_sphere, center=<6.7821 -0.6979 0.3023>}]
 HairStrand_front [{shapeType=Sphere, radius=1.3751, shapeName=dyn_HairStrand_ear_r_02_sphere, center=<2.933 -0.1462 -0.307>}]
 dyn_HairStrand2_l [{shapeType=Sphere, radius=1.3751, shapeName=dyn_HairStrand_ear_r_02_sphere, center=<2.933 -0.1462 -0.307>}]
 dyn_HairStrand_A_r [{shapeType=Sphere, radius=1.3751, shapeName=dyn_HairStrand_ear_r_02_sphere, center=<2.933 -0.1462 -0.307>}]
 dyn_HairStrand_B_r [{shapeType=Sphere, radius=1.3751, shapeName=dyn_HairStrand_ear_r_02_sphere, center=<2.933 -0.1462 -0.307>}]
 dyn_HairStrand_ear_l_01 [{shapeType=Sphere, radius=1.3751, shapeName=dyn_HairStrand_ear_r_02_sphere, center=<2.933 -0.1462 -0.307>}]
 dyn_HairStrand_ear_l_02 [{shapeType=Sphere, radius=1.3751, shapeName=dyn_HairStrand_ear_r_02_sphere, center=<2.933 -0.1462 -0.307>}]
 dyn_HairStrand_ear_r_01 [{shapeType=Sphere, radius=1.6481, shapeName=dyn_HairStrand_ear_r_01_sphere, center=<3.7944 0.2066 0>}]
 dyn_HairStrand_ear_r_02 [{shapeType=Sphere, radius=1.3751, shapeName=dyn_HairStrand_ear_r_02_sphere, center=<2.933 -0.1462 -0.307>}]
 dyn_HairStrand_front_01 [{shapeType=Sphere, radius=1.3751, shapeName=dyn_HairStrand_ear_r_02_sphere, center=<2.933 -0.1462 -0.307>}]
 dyn_HairStrand_front_02 [{shapeType=Sphere, radius=1.3751, shapeName=dyn_HairStrand_ear_r_02_sphere, center=<2.933 -0.1462 -0.307>}]
 dyn_HairStrand_front_03 [{shapeType=Sphere, radius=1.3751, shapeName=dyn_HairStrand_ear_r_02_sphere, center=<2.933 -0.1462 -0.307>}]
 dyn_HairStrand_front_04 [{shapeType=Sphere, radius=1.3751, shapeName=dyn_HairStrand_ear_r_02_sphere, center=<2.933 -0.1462 -0.307>}]
 dyn_HairStrand_l [{shapeType=Sphere, radius=1.3751, shapeName=dyn_HairStrand_ear_r_02_sphere, center=<2.933 -0.1462 -0.307>}]
 dyn_HairStrand_l_01 [{shapeType=Sphere, radius=1.3751, shapeName=dyn_HairStrand_ear_r_02_sphere, center=<2.933 -0.1462 -0.307>}]
 dyn_HairStrand_l_02 [{shapeType=Sphere, radius=1.3751, shapeName=dyn_HairStrand_ear_r_02_sphere, center=<2.933 -0.1462 -0.307>}]
 dyn_ear_l [{shapeType=Sphere, radius=2.708, shapeName=dyn_ear_l_sphere, center=<-3.561 0.42 0.8425>}]
 dyn_ear_r [{shapeType=Sphere, radius=2.708, shapeName=dyn_ear_r_sphere, center=<3.561 -0.42 -0.8425>}]
 dyn_earring [{shapeType=Sphere, radius=1.3234, shapeName=dyn_earring_sphere, center=<2.8438 0 0>}]
}
constraints [
 {bone1Name=dyn_ear_r, bone2Name=FaceAcc_Root, swing1LimitDegrees=3, swing2LimitDegrees=7, twistLimitDegrees=2}
 {bone1Name=dyn_ear_l, bone2Name=FaceAcc_Root, swing1LimitDegrees=3, swing2LimitDegrees=7, twistLimitDegrees=2}
 {
  bone1Name=dyn_earring
  bone2Name=FaceAcc_Root
  swing1LimitDegrees=45
  swing1Motion=Locked
  swing2LimitDegrees=45
  twistLimitDegrees=45
  twistMotion=Locked
 }
 {bone1Name=dyn_HairStrand_ear_r_01, bone2Name=FaceAcc_Root, swing1LimitDegrees=8, swing2LimitDegrees=8, twistLimitDegrees=2}
 {bone1Name=dyn_HairStrand_ear_r_02, bone2Name=dyn_HairStrand_ear_r_01, swing1LimitDegrees=15, swing2LimitDegrees=5, twistLimitDegrees=2}
 {bone1Name=dyn_HairStrand_A_r, bone2Name=FaceAcc_Root, swing1LimitDegrees=15, swing2LimitDegrees=5, twistLimitDegrees=2}
 {bone1Name=dyn_HairStrand_B_r, bone2Name=FaceAcc_Root, swing1LimitDegrees=5, swing2LimitDegrees=15, twistLimitDegrees=2}
 {bone1Name=dyn_HairStrand_front_01, bone2Name=FaceAcc_Root, swing1LimitDegrees=8, swing2LimitDegrees=8, twistLimitDegrees=2}
 {bone1Name=dyn_HairStrand_front_02, bone2Name=dyn_HairStrand_front_01, swing1LimitDegrees=8, swing2LimitDegrees=8, twistLimitDegrees=2}
 {bone1Name=dyn_HairStrand_front_03, bone2Name=dyn_HairStrand_front_02, swing1LimitDegrees=8, swing2LimitDegrees=8, twistLimitDegrees=2}
 {bone1Name=dyn_HairStrand_front_04, bone2Name=dyn_HairStrand_front_02, swing1LimitDegrees=8, swing2LimitDegrees=8, twistLimitDegrees=2}
 {bone1Name=HairStrand_front, bone2Name=FaceAcc_Root, swing1LimitDegrees=8, swing2LimitDegrees=8, twistLimitDegrees=2}
 {bone1Name=dyn_HairStrand_l_01, bone2Name=FaceAcc_Root, swing1LimitDegrees=8, swing2LimitDegrees=8, twistLimitDegrees=2}
 {bone1Name=dyn_HairStrand_l_02, bone2Name=dyn_HairStrand_l_01, swing1LimitDegrees=8, swing2LimitDegrees=8, twistLimitDegrees=2}
 {bone1Name=dyn_HairStrand2_l, bone2Name=FaceAcc_Root, swing1LimitDegrees=8, swing2LimitDegrees=8, twistLimitDegrees=2}
 {bone1Name=dyn_HairStrand_l, bone2Name=FaceAcc_Root, swing1LimitDegrees=8, swing2LimitDegrees=8, twistLimitDegrees=2}
 {bone1Name=dyn_HairStrand_ear_l_01, bone2Name=FaceAcc_Root, swing1LimitDegrees=4, swing2LimitDegrees=4, twistLimitDegrees=2}
 {bone1Name=dyn_HairStrand_ear_l_02, bone2Name=dyn_HairStrand_ear_l_01, swing1LimitDegrees=4, swing2LimitDegrees=4, twistLimitDegrees=2}
]
