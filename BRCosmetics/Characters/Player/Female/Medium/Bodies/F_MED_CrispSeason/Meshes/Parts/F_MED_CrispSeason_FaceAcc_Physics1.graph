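# Physics asset for the CrispSeason medium female character's FaceAcc mesh part, modelling a dynamic braid.
# A chain of six braid bodies, a sphere at the root followed by capsules, links through swing limits that are widest mid-chain, with zero twist limits.
# The braid root attaches to spine_05 with free rotation.
# Capsules on spine_04, spine_05 and the left clavicle add upper-torso shapes, also joined by free-rotation constraints.
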PhysicsAsset /BRCosmetics/Characters/Player/Female/Medium/Bodies/F_MED_CrispSeason/Meshes/Parts/F_MED_CrispSeason_FaceAcc_Physics1
bodies {
 clavicle_l [{shapeType=Capsule, length=6.8064, radius=5.1091, shapeName=clavicle_l_capsule, center=<9.0914 0 0>, rotation=<rot 90 0 0>}]
 dyn_braid_02 [{shapeType=Sphere, radius=1.5331, shapeName=dyn_braid_02_sphere}]
 dyn_braid_03 [{shapeType=Capsule, length=1.8954, radius=2.4496, shapeName=dyn_braid_03_capsule, center=<1.4997 0 0>, rotation=<rot 90 0 0>}]
 dyn_braid_04 [{shapeType=Capsule, length=1.7184, radius=2.3521, shapeName=dyn_braid_04_capsule, center=<1.6622 0 0>, rotation=<rot 90 0 0>}]
 dyn_braid_05 [{
   shapeType=Capsule
   length=5.005
   radius=2.9636
   shapeName=dyn_braid_05_capsule
   center=<1.4129 -0.6321 -0.0071>
   rotation=<rot -84.9238 0.003 -6.737>
  }]
 dyn_braid_06 [{
   shapeType=Capsule
   length=3.7331
   radius=3.3978
   shapeName=dyn_braid_06_capsule
   center=<-0.4523 0.1238 -0.1343>
   rotation=<rot -79.6458 -78.1573 85.2359>
  }]
 dyn_braid_07 [{shapeType=Capsule, length=4.6506, radius=1.6632, shapeName=dyn_braid_07_capsule, center=<2.3055 0 0>, rotation=<rot 90 0 0>}]
 spine_04 [{shapeType=Capsule, length=14.274, radius=10.0647, shapeName=spine_04_capsule, center=<2.0025 0.6238 0>}]
 spine_05 [{shapeType=Capsule, length=14.594, radius=8.1657, shapeName=spine_05_capsule, center=<2.4203 -2.0941 0>}]
}
constraints [
 {bone1Name=dyn_braid_04, bone2Name=dyn_braid_03, swing1LimitDegrees=10, swing2LimitDegrees=10, twistLimitDegrees=0}
 {bone1Name=dyn_braid_05, bone2Name=dyn_braid_04, swing1LimitDegrees=30, swing2LimitDegrees=30, twistLimitDegrees=0}
 {bone1Name=dyn_braid_06, bone2Name=dyn_braid_05, swing1LimitDegrees=30, swing2LimitDegrees=30, twistLimitDegrees=0}
 {bone1Name=dyn_braid_07, bone2Name=dyn_braid_06, swing1LimitDegrees=15, swing2LimitDegrees=15, twistLimitDegrees=0}
 {bone1Name=dyn_braid_03, bone2Name=dyn_braid_02, swing1LimitDegrees=10, swing2LimitDegrees=10, twistLimitDegrees=0}
 {
  bone1Name=dyn_braid_02
  bone2Name=spine_05
  swing1LimitDegrees=45
  swing1Motion=Free
  swing2LimitDegrees=45
  swing2Motion=Free
  twistLimitDegrees=45
  twistMotion=Free
 }
 {
  bone1Name=spine_05
  bone2Name=spine_04
  swing1LimitDegrees=45
  swing1Motion=Free
  swing2LimitDegrees=45
  swing2Motion=Free
  twistLimitDegrees=45
  twistMotion=Free
 }
 {
  bone1Name=clavicle_l
  bone2Name=spine_05
  swing1LimitDegrees=45
  swing1Motion=Free
  swing2LimitDegrees=45
  swing2Motion=Free
  twistLimitDegrees=45
  twistMotion=Free
 }
]
